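# Niagara impact effect for the NobleCrusher pickaxe, with short one-shot bursts tinted by shared system crystal and trim colors.
# Flash_Elements layers flares, stars, spikes and a light on per-element age curves, while Ring expands sprites and meshes and Mist drifts along the system direction vector.
# Front-end scale values resize elements between menu and gameplay.
# In the visible part, physical-surface debris, both smoke emitters and the duplicate Ring001 are disabled.
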NiagaraSystem /BRCosmetics/Weapons/FORT_Melee/Pickaxe_NobleCrusher/FX/NS_Pickaxe_NobleCrusher_Impact
dependencies {
 usedDataInterfaces [
  &NiagaraDataInterfaceCurve
  &NiagaraDataInterfaceMeshRendererInfo
  &NiagaraDataInterfaceVector2DCurve
  &NiagaraDataInterfaceColorCurve
  &NiagaraDataInterfaceVectorCurve
  &NiagaraDataInterfaceArrayFloat
  &NiagaraDataInterfaceArrayFloat3
  /Script/Niagara.NiagaraDataInterfaceArrayFloat4
 ]
 usedDynamicInputs [
  /Game/Effects/Niagara/Modules/Utility/FN_DirectionVectorByPickaxeType
  &MakeFloatFromLinearColor
  /Niagara/DynamicInputs/Math/CompareFloats
 ]
 usedModules [
  &self:SystemSpawnScript.SystemScriptSource.SystemScriptGraph.NiagaraNodeAssignment_1.SetVariables_#1
  &self:SystemSpawnScript.SystemScriptSource.SystemScriptGraph.NiagaraNodeAssignment_0.SetVariables_#2
  &self:SystemSpawnScript.SystemScriptSource.SystemScriptGraph.NiagaraNodeAssignment_4.SetVariables_#3
  &self:SystemSpawnScript.SystemScriptSource.SystemScriptGraph.NiagaraNodeAssignment_6.SetVariables_#4
  &self:SystemSpawnScript.SystemScriptSource.SystemScriptGraph.NiagaraNodeAssignment_8.SetVariables_#5
  &self:SystemSpawnScript.SystemScriptSource.SystemScriptGraph.NiagaraNodeAssignment_9.SetVariables_#6
  &self:SystemSpawnScript.SystemScriptSource.SystemScriptGraph.NiagaraNodeAssignment_10.SetVariables_#7
  &self:SystemSpawnScript.SystemScriptSource.SystemScriptGraph.NiagaraNodeAssignment_11.SetVariables_#8
  &self:SystemSpawnScript.SystemScriptSource.SystemScriptGraph.NiagaraNodeAssignment_3.SetVariables_#9
  &SystemState
  &self:SystemSpawnScript.SystemScriptSource.SystemScriptGraph.NiagaraNodeAssignment_5.SetVariables_#10
  &self:SystemSpawnScript.SystemScriptSource.SystemScriptGraph.NiagaraNodeAssignment_7.SetVariables_#11
 ]
}
emitters {
 FN_PhysMatDebris {
  bEnabled=false
  emitterUpdateScript [
   &EmitterState()
   &SpawnBurst_Instantaneous(
    Spawn Count=&RandomRangeInt(Minimum=5, Maximum=10)
    Use Spawn Probability=false
    UseLoopCountLimit=false
   )
  ]
  particleSpawnScript [
   /SetVariables(Particles.BiasedRandom=&FloatFromCurve(
     FloatCurve=DataInterface() {propertyValues {Curve {keys [
         <at 0.2 1 RCIM_Cubic>
         {arriveTangent=-3.3394, interpMode=RCIM_Cubic, leaveTangent=-3.3394, tangentMode=RCTM_User, time=0.5, value=0.5}
         <at 0.8 0 RCIM_Cubic>
        ]}}}
     CurveIndex=&RandomRangeFloat(Minimum=0, Maximum=1)
    )) {isSetParameters=true}
   &InitializeParticle(
    Color Mode=::NewEnumerator0
    Lifetime=&RandomRangeFloat(Minimum=0.5, Maximum=1.25)
    Mass=&Lerp_Float(A=0.6, B=1.2, Alpha=$Particles.BiasedRandom)
    Mass Mode=::NewEnumerator1
    Mesh Scale=&Multiply_VectorByFloat(
     Vector=&RandomRangeVector(Minimum=<0.9 0.4 0.5>, Maximum=<1 1 1>)
     Float=&Lerp_Float(A=0.7, B=1, Alpha=$Particles.BiasedRandom)
    )
    Mesh Scale Mode=::NewEnumerator4
    UsePositionOffset=false
    Write Material Random=true
   )
   &FN_PhysicalSurface(
    Dirt Mesh Count=2
    Glass Mesh Count=3
    Ice Mesh Count=2
    Metal Gravity=&RandomRangeFloat(Minimum=-350, Maximum=-1800)
    Metal Mesh Count=2
    Metal Velocity=&RandomRangeFloat(Minimum=80, Maximum=375)
    Metal Velocity Cone Angle=130
    Physical Surface=$User.PhysicalSurface
    Stone Gravity=&RandomRangeFloat(Minimum=-700, Maximum=-1250)
    Stone Mesh Count=2
    Stone Velocity=&RandomRangeFloat(Minimum=80, Maximum=550)
    Stone Velocity Cone Angle=120
    Wood Gravity=&RandomRangeFloat(Minimum=-450, Maximum=-900)
    Wood Mesh Count=2
    Wood Scale=1.5
    Wood Velocity=&RandomRangeFloat(Minimum=80, Maximum=600)
    Wood Velocity Cone Angle=130
   )
   &AddVelocityInCone(
    Cone Angle=$Output.FN_PhysicalSurface.SurfaceResponseConeAngle
    Module.Use Velocity Falloff On Cone Axis=false
    Velocity Strength=$Output.FN_PhysicalSurface.SurfaceResponseVelocity
   )
   &InitialMeshOrientation(Use Rotation Vector=false)
   &AddRotationalVelocity(Rotation Rate=&RandomVector(Vector Scale=&RandomRangeFloat(Minimum=1, Maximum=2)))
   &ApplyInitialForces()
   /SetVariables(Particles.GravityOverrideValue=$Output.FN_PhysicalSurface.GravityOverride) {isSetParameters=true}
  ]
  particleUpdateScript [
   &ParticleState()
   &GravityForce(Gravity=$Particles.GravityOverrideValue)
   &Collision(
    Advanced Aging Rate=1
    Age Colliding Particles=false
    CPU Collision Trace Channel=ECC_WorldStatic
    Collision Enabled=&Boolean_AndOperation(
     Bool1=&SetBoolByFloatComparison(A=$Particles.Initial.Lifetime, B=1)
     Bool2=&SetBoolByFloatComparison(A=$Particles.Age, B=0.12)
    )
    Control Roll On Collision=true
    Friction=1
    Friction During a Bounce=0.25
    Max CPU Trace Length=100
    Module.Control Rest State Via Surface Normal=false
    Module.EnableMaxCollisionCount=false
    Module.Randomize Collision Normal Vector=true
    Randomize Collision Normal=0.25
    ScaleRateOfRotation=false
   ) {enabled=false}
   &SolveForcesAndVelocity(Clamp Velocity=false, Limit Acceleration=false)
   &ScaleMeshSize(Scale Factor=&VectorFromFloat(Value=&FloatFromCurve(FloatCurve=DataInterface() {propertyValues {Curve {keys [<at 0.6 1 RCIM_Cubic>, {arriveTangent=-6.308, interpMode=RCIM_Cubic, leaveTangent=-6.308, tangentMode=RCTM_User, time=1, value=0}]}}})))
   &DynamicMaterialParameters(Module.Param0WriteEnabled=true, Module.Param1WriteEnabled=true, Module.Param2WriteEnabled=false, Module.Param3WriteEnabled=false)
  ]
  rendererClasses [&NiagaraSpriteRendererProperties, &NiagaraMeshRendererProperties]
  renderers [
   {rendererClass=&NiagaraSpriteRendererProperties}
   {rendererClass=&NiagaraMeshRendererProperties, rendererIndex=1}
   {rendererClass=&NiagaraMeshRendererProperties, rendererIndex=2}
   {rendererClass=&NiagaraMeshRendererProperties, rendererIndex=3}
   {rendererClass=&NiagaraMeshRendererProperties, rendererIndex=4}
   {rendererClass=&NiagaraMeshRendererProperties, rendererIndex=5}
   {rendererClass=&NiagaraMeshRendererProperties, rendererIndex=6}
   {rendererClass=&NiagaraMeshRendererProperties, rendererIndex=7}
   {rendererClass=&NiagaraMeshRendererProperties, rendererIndex=8}
  ]
 }
 Flash_Elements {
  emitterUpdateScript [
   &EmitterState()
   &SpawnBurst_Instantaneous(Spawn Group=1, Use Spawn Probability=false, UseLoopCountLimit=false) {module=SpawnBurst_Instantaneous001}
   &SpawnBurst_Instantaneous(
    Spawn Count=&RandomRangeInt(Minimum=5, Maximum=7)
    Use Spawn Probability=false
    UseLoopCountLimit=false
   )
  ]
  particleSpawnScript [
   &InitializeParticle(Color Mode=::NewEnumerator0, Lifetime=0.25, Mesh Scale Mode=::NewEnumerator3, UsePositionOffset=false, Write Material Random=true)
   &ShapeLocation(
    Hemisphere Distribution=&FN_FrontEndVector2Modifier(Front End:Vector2f=<1 1>, Gameplay:Vector2f=<1 0.5>)
    MaskSpawnGroup=false
    Override Seed=false
    Rotation Mode=::NewEnumerator1
    Sphere Radius=1
    UseOwnerScale=true
   )
   &SpriteFacingAndAlignment(Sprite Alignment=$Particles.ShapeLocation.ShapeNormal, WriteAlignment=true, WriteFacing=false)
   &FN_PerParticleLOD(Enable SpriteSize Distance Scaling=true, Sprite Size Scale Far=<2.5 2.5>)
   /SetVariables(Particles.SpriteSize=<1 1>, Particles.VisibilityTag=&SpawnGroup()) {isSetParameters=true}
   /SetVariables(
    Particles.Flash Spike Lifetime:NiagaraFloat=&RandomRangeFloat(Minimum=0.08, Maximum=0.09)
    Particles.Flash Spike Sprite Size:Vector2f=&FN_FrontEndVector2Modifier(
     Front End:Vector2f=&RandomRangeVector2D(Minimum=<1.4 25>, Maximum=<4.9 40>)
     Gameplay:Vector2f=&RandomRangeVector2D(Minimum=<2.1 45>, Maximum=<7 55>)
    )
   ) {isSetParameters=true}
   /SetVariables(
    Particles.Flash Star Cam Offset=15
    Particles.Flash Star Lifetime=0.17
    Particles.Flash Star Rotation:NiagaraFloat=&RandomRangeFloat(Minimum=0, Maximum=360)
    Particles.Flash Star Sprite Size:Vector2f=&Multiply_Vector2D_ByFloat(
     Float=&Multiply_Float(A=$System.FrontendScale, B=&RandomRangeFloat(Minimum=150, Maximum=170))
     Vector2D=$Particles.SpriteSize
    )
    Particles.Flash Star SubUV Index:NiagaraFloat=&RandomRangeFloat(Minimum=0, Maximum=3)
   ) {isSetParameters=true}
   /SetVariables(
    Particles.Dark Flare Sprite Size:Vector2f=&Vector2DFromFloat(Value=&Multiply_Float(A=$System.FrontendScale, B=112))
    Particles.Flash Flare Camera Offset=26
    Particles.Flash Flare Lifetime:NiagaraFloat=&FN_FrontEndFloatModifier(Gameplay=0.2, Front End=0.125)
    Particles.Flash Flare Sprite Size:Vector2f=&FN_FrontEndVector2Modifier(Front End:Vector2f=<90 90>, Gameplay:Vector2f=<160 160>)
   ) {isSetParameters=true}
   &DynamicMaterialParameters(
    Index 0 Param 1=0.2
    Index 0 Param 2=1
    Module.Index0Param0Write=true
    Module.Index0Param1Write=true
    Module.Index0Param2Write=true
    Module.Index0Param3Write=true
   )
   &InitialMeshOrientation(Orientation Axis=<0 0 1>, Orientation Vector=<0 0 1>, Rotation=<0 0.25 0>, Use Orientation Vector=true, Use Rotation Vector=true)
   &InitialMeshOrientation(
    Orientation Axis=<0 0 0>
    Orientation Vector=<0 0 0>
    Rotation=&RandomRangeVector(Minimum=<0 0 0>, Maximum=<1 0 0>)
    Use Orientation Vector=false
    Use Rotation Vector=true
   ) {module=InitialMeshOrientation001}
  ]
  particleUpdateScript [
   &ParticleState()
   /SetVariables(
    Particles.Update Flash Flare Age:NiagaraFloat=&Divide_Float(A=$Particles.Age, B=$Particles.Initial.Flash Flare Lifetime)
    Particles.Update Flash Spike Age:NiagaraFloat=&Divide_Float(A=$Particles.Age, B=$Particles.Initial.Flash Spike Lifetime)
    Particles.Update Flash Star Age:NiagaraFloat=&Divide_Float(A=$Particles.Age, B=$Particles.Initial.Flash Star Lifetime)
   ) {isSetParameters=true}
   /SetVariables(
    Particles.Dark Flare Color:LinearColor=&MakeLinearColorFromVectorAndFloat(
     "Vector (RGB)"=<0 0 0>
     "Float (Alpha)"=&FloatFromCurve(
      FloatCurve=DataInterface() {propertyValues {Curve {
         defaultValue=1
         keys [
          {arriveTangent=10.726, interpMode=RCIM_Cubic, leaveTangent=10.726, tangentMode=RCTM_User, time=0, value=0}
          {arriveTangent=-0.8745, interpMode=RCIM_Cubic, leaveTangent=-0.8745, tangentMode=RCTM_User, time=0.2015, value=0.9847}
          {arriveTangent=-1.2692, interpMode=RCIM_Cubic, leaveTangent=-1.2692, tangentMode=RCTM_User, time=1, value=0}
         ]
        }}}
      CurveIndex=$Particles.Update Flash Flare Age
      Scale Curve=2
     )
    )
    Particles.Update Flash Spike Color:LinearColor=&MakeLinearColorFromVectorAndFloat(
     "Vector (RGB)"=&MakeVectorFromLinearColorRGB(Color=&MakeCustomLinearColorFromBool(Bool=&RandomBool(), True Color=$System.SystemCrystalCol2, False Color=$System.SystemTrimCol))
     "Float (Alpha)"=&FloatFromCurve(
      FloatCurve=DataInterface() {propertyValues {Curve {
         defaultValue=1
         keys [
          {arriveTangent=9.0725, interpMode=RCIM_Cubic, leaveTangent=9.0725, tangentMode=RCTM_User, time=0, value=0}
          {arriveTangent=-1.9406, interpMode=RCIM_Cubic, leaveTangent=-1.9406, tangentMode=RCTM_User, time=0.1408, value=0.9847}
          {arriveTangent=-0.0405, interpMode=RCIM_Cubic, leaveTangent=-0.0405, tangentMode=RCTM_User, time=1, value=0}
         ]
        }}}
      CurveIndex=$Particles.Update Flash Spike Age
     )
    )
    Particles.Update Flash Spike Sprite Size:Vector2f=&Multiply_Vector2D_ByFloat(
     Float=&FloatFromCurve(FloatCurve=DataInterface() {propertyValues {Curve {keys [<at 0 1.5>,<at 1 2>]}}}, CurveIndex=$Particles.Update Flash Spike Age)
     Vector2D=$Particles.Initial.Flash Spike Sprite Size
    )
    Particles.Update Flash Star Color:LinearColor=&MakeLinearColorFromVectorAndFloat(
     "Vector (RGB)"=&MakeVectorFromLinearColorRGB(Color=$System.SystemCrystalCol)
     "Float (Alpha)"=&FloatFromCurve(
      FloatCurve=DataInterface() {propertyValues {Curve {
         defaultValue=1
         keys [
          {arriveTangent=-0.0179, interpMode=RCIM_Cubic, leaveTangent=-0.0179, tangentMode=RCTM_User, time=0.15, value=0.8}
          {arriveTangent=-1.1774, interpMode=RCIM_Cubic, leaveTangent=-1.1774, tangentMode=RCTM_User, time=0.6632, value=0.5329}
          {arriveTangent=-2.2282, interpMode=RCIM_Cubic, leaveTangent=-2.2282, tangentMode=RCTM_User, time=1, value=0}
         ]
        }}}
      CurveIndex=$Particles.Update Flash Star Age
     )
    )
    Particles.Update Flash Star Sprize Size:Vector2f=&Multiply_Vector2D_ByFloat(
     Float=&FloatFromCurve(
      FloatCurve=DataInterface() {propertyValues {Curve {keys [
          {arriveTangent=2.7013, interpMode=RCIM_Cubic, leaveTangent=2.7013, tangentMode=RCTM_User, time=0, value=0}
          {interpMode=RCIM_Cubic, tangentMode=RCTM_User, time=1, value=1}
         ]}}}
      CurveIndex=$Particles.Update Flash Star Age
     )
     Vector2D=$Particles.Initial.Flash Star Sprite Size
    )
   ) {isSetParameters=true}
   /SetVariables(
    Particles.Dark Flare Color:LinearColor=&MakeLinearColorFromVectorAndFloat(
     "Vector (RGB)"=<0 0 0>
     "Float (Alpha)"=&FloatFromCurve(
      FloatCurve=DataInterface() {propertyValues {Curve {
         defaultValue=1
         keys [
          {arriveTangent=10.726, interpMode=RCIM_Cubic, leaveTangent=10.726, tangentMode=RCTM_User, time=0, value=0}
          {arriveTangent=-0.8745, interpMode=RCIM_Cubic, leaveTangent=-0.8745, tangentMode=RCTM_User, time=0.2015, value=0.9847}
          {arriveTangent=-1.2692, interpMode=RCIM_Cubic, leaveTangent=-1.2692, tangentMode=RCTM_User, time=1, value=0}
         ]
        }}}
      CurveIndex=$Particles.Update Flash Flare Age
      Scale Curve=1.5
     )
    )
    Particles.PsySize:Vector2f=&Vector2DFromFloat(Value=&Multiply_Float(
      A=$System.FrontendScale
      B=&FloatFromCurve(
       FloatCurve=DataInterface() {propertyValues {Curve {
          defaultValue=1
          keys [
           {arriveTangent=4.1785, interpMode=RCIM_Cubic, leaveTangent=4.1785, tangentMode=RCTM_User, time=0, value=0}
           {arriveTangent=1.1179, interpMode=RCIM_Cubic, leaveTangent=1.1179, tangentMode=RCTM_User, time=0.2042, value=0.6228}
           {arriveTangent=0.0224, interpMode=RCIM_Cubic, leaveTangent=0.0224, tangentMode=RCTM_User, time=1, value=1}
          ]
         }}}
       CurveIndex=$Particles.Update Flash Flare Age
       Scale Curve=94.9991
      )
     ))
    Particles.Update Flash Spike Color:LinearColor=&MakeLinearColorFromVectorAndFloat(
     "Vector (RGB)"=&MakeVectorFromLinearColorRGB(Color=$System.SystemCrystalCol2)
     "Float (Alpha)"=&FloatFromCurve(
      FloatCurve=DataInterface() {propertyValues {Curve {
         defaultValue=1
         keys [
          {arriveTangent=9.0725, interpMode=RCIM_Cubic, leaveTangent=9.0725, tangentMode=RCTM_User, time=0, value=0}
          {arriveTangent=-1.9406, interpMode=RCIM_Cubic, leaveTangent=-1.9406, tangentMode=RCTM_User, time=0.1408, value=0.9847}
          {arriveTangent=-0.0405, interpMode=RCIM_Cubic, leaveTangent=-0.0405, tangentMode=RCTM_User, time=1, value=0}
         ]
        }}}
      CurveIndex=$Particles.Update Flash Spike Age
     )
    )
    Particles.Update Flash Spike Sprite Size:Vector2f=&Multiply_Vector2D_ByFloat(
     Float=&FloatFromCurve(FloatCurve=DataInterface() {propertyValues {Curve {keys [<at 0 1.5>,<at 1 2>]}}}, CurveIndex=$Particles.Update Flash Spike Age)
     Vector2D=$Particles.Initial.Flash Spike Sprite Size
    )
    Particles.Update Flash Star Color:LinearColor=&MakeLinearColorFromVectorAndFloat(
     "Vector (RGB)"=&MakeVectorFromLinearColorRGB(Color=$System.SystemCrystalCol)
     "Float (Alpha)"=&FloatFromCurve(
      FloatCurve=DataInterface() {propertyValues {Curve {
         defaultValue=1
         keys [
          {arriveTangent=-0.018, interpMode=RCIM_Cubic, leaveTangent=-0.018, tangentMode=RCTM_User, time=0.15, value=0.8}
          {arriveTangent=-1.1774, interpMode=RCIM_Cubic, leaveTangent=-1.1774, tangentMode=RCTM_User, time=0.6632, value=0.5329}
          {arriveTangent=-2.2282, interpMode=RCIM_Cubic, leaveTangent=-2.2282, tangentMode=RCTM_User, time=1, value=0}
         ]
        }}}
      CurveIndex=$Particles.Update Flash Star Age
     )
    )
    Particles.Update Flash Star Sprize Size:Vector2f=&Multiply_Vector2D_ByFloat(
     Float=&FloatFromCurve(
      FloatCurve=DataInterface() {propertyValues {Curve {keys [
          {arriveTangent=2.7013, interpMode=RCIM_Cubic, leaveTangent=2.7013, tangentMode=RCTM_User, time=0, value=0}
          {interpMode=RCIM_Cubic, tangentMode=RCTM_User, time=1, value=1}
         ]}}}
      CurveIndex=$Particles.Update Flash Star Age
     )
     Vector2D=$Particles.Initial.Flash Star Sprite Size
    )
    Particles.Update Flash Star Sprize Size Alt:Vector2f=&Multiply_Vector2D_ByFloat(
     Float=&FloatFromCurve(
      FloatCurve=DataInterface() {propertyValues {Curve {keys [
          {arriveTangent=2.7013, interpMode=RCIM_Cubic, leaveTangent=2.7013, tangentMode=RCTM_User, time=0, value=0}
          {interpMode=RCIM_Cubic, tangentMode=RCTM_User, time=1, value=1}
         ]}}}
      CurveIndex=$Particles.Update Flash Star Age
     )
     Vector2D=$Particles.Initial.Flash Star Sprite Size Alt
    )
   ) {enabled=false, isSetParameters=true}
   /SetVariables(
    Particles.PsyColor:LinearColor=&MakeLinearColorFromVectorAndFloat(
     "Vector (RGB)"=&MakeVectorFromLinearColorRGB(Color=$System.SystemCrystalCol2)
     "Float (Alpha)"=&FloatFromCurve(
      FloatCurve=DataInterface() {propertyValues {Curve {
         defaultValue=1
         keys [
          {arriveTangent=7.0929, interpMode=RCIM_Cubic, leaveTangent=7.0929, tangentMode=RCTM_User, time=0, value=0.1189}
          {arriveTangent=0.0107, interpMode=RCIM_Cubic, leaveTangent=0.0107, tangentMode=RCTM_User, time=0.2866, value=0.9847}
          {arriveTangent=-0.4679, interpMode=RCIM_Cubic, leaveTangent=-0.4679, tangentMode=RCTM_User, time=1, value=0}
         ]
        }}}
      CurveIndex=$Particles.Update Flash Flare Age
     )
    )
    Particles.Update Flash Flare Color:LinearColor=&Multiply_LinearColorByFloat(
     Linear Color=$System.SystemCrystalCol2
     Float=&FloatFromCurve(
      FloatCurve=DataInterface() {propertyValues {Curve {
         defaultValue=1
         keys [
          {arriveTangent=9.0725, interpMode=RCIM_Cubic, leaveTangent=9.0725, tangentMode=RCTM_User, time=0, value=0.3282}
          {arriveTangent=-0.8745, interpMode=RCIM_Cubic, leaveTangent=-0.8745, tangentMode=RCTM_User, time=0.1842, value=0.9847}
          {arriveTangent=-1.2692, interpMode=RCIM_Cubic, leaveTangent=-1.2692, tangentMode=RCTM_User, time=1, value=0}
         ]
        }}}
      CurveIndex=$Particles.Update Flash Flare Age
      Scale Curve=3.3795
     )
    )
   ) {isSetParameters=true}
   /SetVariables(Particles.WaveParam:Vector4f=&MakeVector4(
     X=&FloatFromCurve(
      FloatCurve=DataInterface() {propertyValues {Curve {keys [{arriveTangent=-0.2215, interpMode=RCIM_Cubic, leaveTangent=-0.2215, tangentMode=RCTM_User, time=0, value=0.1}, <at 1 0 RCIM_Cubic>]}}}
      CurveIndex=$Particles.Update Flash Flare Age
     )
     Y=&FloatFromCurve(
      FloatCurve=DataInterface() {propertyValues {Curve {keys [{arriveTangent=-0.4676, interpMode=RCIM_Cubic, leaveTangent=-0.4676, tangentMode=RCTM_User, time=0, value=1}, <at 1 0.7 RCIM_Cubic>]}}}
      CurveIndex=$Particles.Update Flash Flare Age
     )
     Z=0.5
     W=0
    )) {isSetParameters=true}
   &FN_Light_Attributes(
    Color=<rgba 1 0.3626 0 1>
    Create Color Attribute=true
    Create Exponent Attribute=true
    Create Position Attribute=true
    Create Radius Attribute=true
    Exponent=5
    Position=<15 0 0>
    Radius=90
    Scale Color=&FloatFromCurve(
     FloatCurve=DataInterface() {propertyValues {Curve {keys [
         {interpMode=RCIM_Cubic, tangentMode=RCTM_User, time=0, value=1}
         {arriveTangent=-0.5426, interpMode=RCIM_Cubic, leaveTangent=-0.5426, tangentMode=RCTM_User, time=1, value=0}
        ]}}}
     Scale Curve=10
    )
   )
  ]
  rendererClasses [
   &NiagaraSpriteRendererProperties
   &NiagaraLightRendererProperties
   &NiagaraMeshRendererProperties
  ]
  renderers [
   {rendererClass=&NiagaraSpriteRendererProperties}
   {rendererClass=&NiagaraSpriteRendererProperties, rendererIndex=1}
   {rendererClass=&NiagaraSpriteRendererProperties, rendererIndex=2}
   {rendererClass=&NiagaraLightRendererProperties, rendererIndex=3}
   {rendererClass=&NiagaraSpriteRendererProperties, rendererIndex=4}
   {rendererClass=&NiagaraMeshRendererProperties, rendererIndex=5}
  ]
 }
 Impact_Smoke {
  bEnabled=false
  emitterUpdateScript [
   &EmitterState()
   &SpawnBurst_Instantaneous(
    Spawn Count=&RandomRangeInt(Minimum=2, Maximum=3)
    Use Spawn Probability=false
    UseLoopCountLimit=false
   )
  ]
  particleSpawnScript [
   &InitializeParticle(
    Color Maximum=<rgba 0.76 1 1 0.2>
    Color Minimum=<rgba 0.8 0.917 1 0.15>
    Color Mode=::NewEnumerator2
    Lifetime Max=0.4
    Lifetime Min=0.3
    Lifetime Mode=::NewEnumerator1
    Sprite Size Mode=::NewEnumerator3
    Uniform Sprite Size=&Multiply_Float(
     A=&MakeCustomFloatFromBool(Bool=$User.bIsFrontEnd, True Float=0.45, False Float=1)
     B=&RandomRangeFloat(Minimum=100, Maximum=130)
    )
    UsePositionOffset=false
    Write Material Random=true
   )
   &CameraOffset(Camera Offset Amount=&RandomRangeFloat(Minimum=40, Maximum=50, Evaluation Type=::NewEnumerator1))
   &ShapeLocation(
    Hemisphere Distribution=<1 0.5>
    MaskSpawnGroup=false
    Override Seed=false
    Rotation Angle=90
    Rotation Mode=::NewEnumerator1
    Sphere Radius=&MakeCustomFloatFromBool(Bool=$User.bIsFrontEnd, True Float=5, False Float=10)
    UseOwnerScale=true
   )
   &AddVelocity(Velocity=&Multiply_VectorByFloat(Vector=$Particles.ShapeLocation.ShapeVector, Float=10))
  ]
  particleUpdateScript [
   &ParticleState()
   &ScaleSpriteSize(Uniform Curve Sprite Scale=DataInterface() {propertyValues {Curve {keys [
        {arriveTangent=1.7761, interpMode=RCIM_Cubic, leaveTangent=1.7761, tangentMode=RCTM_User, time=0, value=0}
        {interpMode=RCIM_Cubic, tangentMode=RCTM_User, time=1, value=1}
       ]}}})
   &ScaleColor(
    Scale Alpha=&FloatFromCurve(FloatCurve=DataInterface() {propertyValues {Curve {keys [
         {arriveTangent=2.4358, interpMode=RCIM_Cubic, leaveTangent=2.4358, tangentMode=RCTM_User, time=0, value=0}
         <at 0.25 0.5 RCIM_Cubic>
         <at 1 0 RCIM_Cubic>
        ]}}})
    Scale RGB=&VectorFromCurve(VectorCurve=DataInterface() {propertyValues {XCurve {keys [<at 0 1>,<at 1 1>]}, YCurve {keys [<at 0 1>,<at 1 1>]}, ZCurve {keys [<at 0 1>,<at 1 1>]}}})
    ScaleA=true
    ScaleRGB=true
   )
   &DynamicMaterialParameters(
    Index 0 Param 1=&Add_Float(
     A=&RandomRangeFloat(Minimum=0, Maximum=1)
     B=&FloatFromCurve(
      FloatCurve=DataInterface() {propertyValues {Curve {keys [
          <at 0 0 RCIM_Cubic>
          {arriveTangent=1.5947, interpMode=RCIM_Cubic, leaveTangent=1.5947, tangentMode=RCTM_User, time=0.2557, value=0.5257}
          <at 1 1 RCIM_Cubic>
         ]}}}
      CurveIndex=$Particles.Age
      Scale Curve=0.25
     )
    )
    Module.Param0WriteEnabled=true
    Module.Param1WriteEnabled=true
    Module.Param2WriteEnabled=true
    Module.Param3WriteEnabled=true
   ) {module=DynamicMaterialParameters001}
   &SolveForcesAndVelocity(Clamp Velocity=false, Limit Acceleration=false)
  ]
  rendererClasses [&NiagaraSpriteRendererProperties]
  renderers [
   {rendererClass=&NiagaraSpriteRendererProperties}
   {rendererClass=&NiagaraSpriteRendererProperties, rendererIndex=1}
  ]
 }
 Impact_Smoke001 {
  bEnabled=false
  emitterUpdateScript [
   &EmitterState()
   &SpawnBurst_Instantaneous(
    Spawn Count=&RandomRangeInt(Minimum=2, Maximum=3)
    Use Spawn Probability=false
    UseLoopCountLimit=false
   )
  ]
  particleSpawnScript [
   &InitializeParticle(
    Color Maximum=<rgba 0.76 1 1 0.2>
    Color Minimum=<rgba 0.8 0.917 1 0.15>
    Color Mode=::NewEnumerator2
    Sprite Size Mode=::NewEnumerator3
    Uniform Sprite Size=&Multiply_Float(
     A=&MakeCustomFloatFromBool(Bool=$User.bIsFrontEnd, True Float=0.45, False Float=1)
     B=&RandomRangeFloat(Minimum=100, Maximum=130)
    )
    UsePositionOffset=false
    Write Material Random=true
   )
   &CameraOffset(Camera Offset Amount=&RandomRangeFloat(Minimum=40, Maximum=50, Evaluation Type=::NewEnumerator1)) {enabled=false}
   &ShapeLocation(
    Hemisphere Distribution=<1 0.5>
    MaskSpawnGroup=false
    Override Seed=false
    Rotation Angle=90
    Rotation Mode=::NewEnumerator1
    Sphere Radius=&MakeCustomFloatFromBool(Bool=$User.bIsFrontEnd, True Float=5, False Float=10)
    UseOwnerScale=true
   ) {enabled=false}
   &AddVelocity(Velocity=&Multiply_VectorByFloat(Vector=$Particles.ShapeLocation.ShapeVector, Float=10)) {enabled=false}
  ]
  particleUpdateScript [
   &ParticleState() {enabled=false}
   &ScaleSpriteSize(Uniform Curve Sprite Scale=DataInterface() {propertyValues {Curve {keys [
        {arriveTangent=1.7761, interpMode=RCIM_Cubic, leaveTangent=1.7761, tangentMode=RCTM_User, time=0, value=0}
        {interpMode=RCIM_Cubic, tangentMode=RCTM_User, time=1, value=1}
       ]}}}) {enabled=false}
   &ScaleColor(
    Scale Alpha=&FloatFromCurve(FloatCurve=DataInterface() {propertyValues {Curve {keys [
         {arriveTangent=2.4358, interpMode=RCIM_Cubic, leaveTangent=2.4358, tangentMode=RCTM_User, time=0, value=0}
         <at 0.25 0.5 RCIM_Cubic>
         <at 1 0 RCIM_Cubic>
        ]}}})
    Scale RGB=&VectorFromCurve(VectorCurve=DataInterface() {propertyValues {XCurve {keys [<at 0 1>,<at 1 1>]}, YCurve {keys [<at 0 1>,<at 1 1>]}, ZCurve {keys [<at 0 1>,<at 1 1>]}}})
    ScaleA=true
    ScaleRGB=true
   ) {enabled=false}
   &DynamicMaterialParameters(
    Index 0 Param 1=&Add_Float(
     A=&RandomRangeFloat(Minimum=0, Maximum=1)
     B=&FloatFromCurve(
      FloatCurve=DataInterface() {propertyValues {Curve {keys [
          <at 0 0 RCIM_Cubic>
          {arriveTangent=1.5947, interpMode=RCIM_Cubic, leaveTangent=1.5947, tangentMode=RCTM_User, time=0.2557, value=0.5257}
          <at 1 1 RCIM_Cubic>
         ]}}}
      CurveIndex=$Particles.Age
      Scale Curve=0.25
     )
    )
    Module.Param0WriteEnabled=true
    Module.Param1WriteEnabled=true
    Module.Param2WriteEnabled=true
    Module.Param3WriteEnabled=true
   ) {enabled=false, module=DynamicMaterialParameters001}
   &SolveForcesAndVelocity(Clamp Velocity=false, Limit Acceleration=false) {enabled=false}
  ]
  rendererClasses [&NiagaraSpriteRendererProperties]
  renderers [
   {rendererClass=&NiagaraSpriteRendererProperties}
   {rendererClass=&NiagaraSpriteRendererProperties, rendererIndex=1}
  ]
 }
 Mist {
  emitterUpdateScript [
   &EmitterState()
   &SpawnBurst_Instantaneous(Spawn Count=3, Use Spawn Probability=false, UseLoopCountLimit=false)
  ]
  particleSpawnScript [
   &InitializeParticle(
    Color=&MultiplyLinearColor(
     Linear Color=$System.SystemCrystalCol
     Scale Factor=/Niagara/DynamicInputs/UniformRange/V2/RandomRangeVector4(
      Minimum:Vector4f=<0.5 0.5 0.5 0.15>
      Maximum:Vector4f=<0.5 0.5 0.5 0.35>
      Evaluation Type:ENiagaraRandomnessEvaluation=::NewEnumerator0
      Recalculate Random Each Loop:NiagaraBool=true
      Randomness Mode:ENiagaraRandomnessMode=::NewEnumerator0
      Random Seed:NiagaraInt32=0
      Fixed Random Seed:NiagaraBool=false
     )
    )
    Lifetime=&RandomRangeFloat(Minimum=0.4, Maximum=0.6)
    Sprite Size Mode=::NewEnumerator3
    Sprite UV Mode=::NewEnumerator4
    Sprite UV Scale:Vector2f=<1 -1>
    Uniform Sprite Size=&Multiply_Float(A=&RandomRangeFloat(Minimum=50, Maximum=75), B=$System.FrontEndScale)
    UsePositionOffset=false
    Write Material Random=true
   )
   &ShapeLocation(
    MaskSpawnGroup=false
    Offset=<0 -1.3139 0>
    Offset Mode=::NewEnumerator0
    Override Seed=false
    Rotation Angle=0.25
    Rotation Axis=<0 1 0>
    Rotation Mode=::NewEnumerator1
    Sphere Radius=10
    Sphere Surface Distribution=1
    UseOwnerScale=true
   )
   &AddVelocity(Velocity=$System.DirectionVector, Velocity Speed Scale=&Multiply_Float(A=120, B=$System.FrontEndScale))
   &AddVelocity(
    Velocity=&RandomRangeVector(Minimum=<-1 -1 -1>, Maximum=<1 1 1>)
    Velocity Speed Scale=&Multiply_Float(A=90, B=$System.FrontEndScale)
   ) {module=AddVelocity001}
  ]
  particleUpdateScript [
   &ParticleState()
   &Drag(Use Linear Drag=true, Use Rotational Drag=true)
   &SolveForcesAndVelocity(Clamp Velocity=false, Limit Acceleration=false)
   &ScaleColor(
    Scale Alpha=&FloatFromCurve(FloatCurve=DataInterface() {propertyValues {Curve {
        defaultValue=1
        keys [
         {arriveTangent=9.7134, interpMode=RCIM_Cubic, leaveTangent=9.7134, tangentMode=RCTM_User, time=0, value=0}
         {arriveTangent=0.0707, interpMode=RCIM_Cubic, leaveTangent=0.0707, tangentMode=RCTM_User, time=0.3002, value=0.8683}
         {arriveTangent=-2.3231, interpMode=RCIM_Cubic, leaveTangent=-2.3231, tangentMode=RCTM_User, time=1, value=0}
        ]
       }}})
    ScaleA=true
    ScaleRGB=true
   )
   &ScaleSpriteSize(
    Non-Uniform Curve Sprite Scale=DataInterface() {propertyValues {
      XCurve {keys [{tangentMode=RCTM_User, time=0, value=1}, {tangentMode=RCTM_User, time=1, value=0.8033}]}
      YCurve {keys [{tangentMode=RCTM_User, time=0, value=1}, {tangentMode=RCTM_User, time=1, value=1.1869}]}
     }}
    Scale Sprite Size Mode=::NewEnumerator3
   )
   &CameraOffset(Camera Offset Amount=&FloatFromCurve(FloatCurve=DataInterface() {propertyValues {Curve {keys [<at 0 20>,<at 1 100>]}}}))
   /SetVariables(Particles.SubImageIndex=&FloatFromCurve(FloatCurve=DataInterface() {propertyValues {Curve {keys [
         {arriveTangent=131.9686, interpMode=RCIM_Cubic, leaveTangent=131.9686, tangentMode=RCTM_User, time=0, value=0}
         {arriveTangent=50.2591, interpMode=RCIM_Cubic, leaveTangent=50.2591, tangentMode=RCTM_User, time=1, value=64}
        ]}}})) {isSetParameters=true}
   &DynamicMaterialParameters(
    Index 0 Param 1=0.2
    Index 0 Param 2=1.3
    Module.Index0Float0Write=true
    Module.Index0Float1Write=true
    Module.Index0Float2Write=true
    Module.Index0Float3Write=true
   )
  ]
  rendererClasses [&NiagaraSpriteRendererProperties]
  renderers [{rendererClass=&NiagaraSpriteRendererProperties}]
 }
 Ring {
  emitterUpdateScript [
   &EmitterState()
   &SpawnBurst_Instantaneous(Use Spawn Probability=false, UseLoopCountLimit=false)
   &SpawnBurst_Instantaneous(Spawn Group=1, Use Spawn Probability=false, UseLoopCountLimit=false) {module=SpawnBurst_Instantaneous001}
   &SpawnBurst_Instantaneous(Spawn Group=2, Use Spawn Probability=false, UseLoopCountLimit=false) {module=SpawnBurst_Instantaneous002}
  ]
  particleSpawnScript [
   /SetVariables(Particles.VisibilityTag=&SpawnGroup()) {isSetParameters=true, module=SetVariables_#12}
   &InitializeParticle(
    Color=&MultiplyLinearColor(
     Linear Color=$System.SystemTrimCol
     Scale Factor=/Niagara/DynamicInputs/Arrays/SelectVector4_FromArray(
      Array Sampling Mode:ENiagaraArraySamplingMode=::NewEnumerator1
      Direct Array Index:NiagaraInt32=&SpawnGroup()
      Interpolation:NiagaraFloat=?
      Direct Array Mode:ENiagara_WrapClamp=::NewEnumerator0
      Evaluation Type:ENiagaraRandomnessEvaluation=::NewEnumerator0
      Normalized Interpolation:NiagaraBool=true
      Recalculate Random Each Loop:NiagaraBool=false
      Randomness Mode:ENiagaraRandomnessMode=::NewEnumerator0
      Random Seed:NiagaraInt32=?
      Fixed Random Seed:NiagaraBool=false
      Vector4 Selection Array:NiagaraDataInterfaceArrayFloat4=DataInterface() {propertyValues {GpuSyncMode=SyncCpuToGpu, MaxElements=0, FloatData [<2 2 2 0.823>, <2.4 2.4 2.4 10>, <1.15 1.15 1.15 0.6>]}}
     )
    )
    Lifetime=0.26
    Mesh Scale Mode=::NewEnumerator3
    Mesh Uniform Scale=1.2
    Sprite Rotation Mode=::NewEnumerator2
    Sprite Size Mode=::NewEnumerator3
    Uniform Sprite Size=&MaskFloatBySpawnGroup(Masked Float=100, Passthrough Float=90)
    UsePositionOffset=false
    Write Material Random=true
   )
   &CameraOffset(Camera Offset Amount=12)
   &InitialMeshOrientation(Orientation Axis=<0 0 1>, Orientation Vector=<0 0 1>, Rotation=<0 0.25 0>, Use Orientation Vector=true, Use Rotation Vector=true)
   &InitialMeshOrientation(
    Orientation Axis=<0 0 0>
    Orientation Vector=<0 0 0>
    Rotation=&RandomRangeVector(Minimum=<0 0 0>, Maximum=<1 0 0>)
    Use Orientation Vector=false
    Use Rotation Vector=true
   ) {module=InitialMeshOrientation001}
   /SetVariables(Particles.WaveParam:Vector4f=<0.3 1 1 1>) {isSetParameters=true, module=SetVariables_#13}
   /SetVariables() {enabled=false, isSetParameters=true, module=SetVariables_#14}
  ]
  particleUpdateScript [
   &ParticleState()
   &ScaleColor(
    Scale Alpha=&FloatFromCurve(FloatCurve=DataInterface() {propertyValues {Curve {keys [
         {arriveTangent=2.8157, interpMode=RCIM_Cubic, leaveTangent=2.8157, tangentMode=RCTM_User, time=0, value=0.2235}
         {arriveTangent=-0.2235, interpMode=RCIM_Cubic, leaveTangent=-0.2235, time=0.0858, value=0.4956}
         {arriveTangent=-0.0522, interpMode=RCIM_Cubic, leaveTangent=-0.0522, tangentMode=RCTM_User, time=1, value=0}
        ]}}})
    ScaleA=true
    ScaleRGB=true
   )
   &ScaleSpriteSize(
    Uniform Curve Scale=$System.FrontendScale
    Uniform Curve Sprite Scale=DataInterface() {propertyValues {Curve {keys [
        {
         arriveTangent=0.7015
         arriveTangentWeight=0.3573
         interpMode=RCIM_Cubic
         leaveTangent=0.7015
         leaveTangentWeight=0.3573
         tangentMode=RCTM_User
         tangentWeightMode=RCTWM_WeightedBoth
         time=0
         value=0.7882
        }
        {
         arriveTangent=0.0045
         arriveTangentWeight=0.7576
         interpMode=RCIM_Cubic
         leaveTangent=0.0045
         leaveTangentWeight=0.7576
         tangentMode=RCTM_User
         tangentWeightMode=RCTWM_WeightedBoth
         time=1
         value=1
        }
       ]}}}
   )
   &ScaleMeshSize(Scale Factor=&VectorFromFloat(Value=&FloatFromCurve(
      FloatCurve=DataInterface() {propertyValues {Curve {
         defaultValue=1
         keys [
          {arriveTangent=7.2689, interpMode=RCIM_Cubic, leaveTangent=7.2689, tangentMode=RCTM_User, time=0, value=0.2235}
          {arriveTangent=0.4035, interpMode=RCIM_Cubic, leaveTangent=0.4035, tangentMode=RCTM_User, time=0.173, value=0.8111}
          {arriveTangent=0.2855, interpMode=RCIM_Cubic, leaveTangent=0.2855, tangentMode=RCTM_User, time=1, value=1}
         ]
        }}}
      Scale Curve=$System.FrontendScale
     )))
   /SetVariables(Particles.RadialColor:LinearColor=&MultiplyLinearColor(Linear Color=$System.SystemTrimCol, Scale Factor=<1 1 1 0.4>)) {enabled=false, isSetParameters=true, module=SetVariables_#15}
   /SetVariables(Particles.RadialParam:Vector4f=&MakeVector4(X=&Lerp_Float(A=-1, B=1.2704, Alpha=$Particles.NormalizedAge), Y=2, Z=3, W=0)) {enabled=false, isSetParameters=true, module=SetVariables_#16}
   /SetVariables(Particles.RadialParam:Vector4f=&MakeVector4(
     X=&Lerp_Float(
      A=-1.5328
      B=-1.2
      Alpha=&FloatFromCurve(FloatCurve=DataInterface() {propertyValues {Curve {
          defaultValue=1
          keys [
           {arriveTangent=2.0823, interpMode=RCIM_Cubic, leaveTangent=2.0823, tangentMode=RCTM_User, time=0, value=0}
           {arriveTangent=0.0224, interpMode=RCIM_Cubic, leaveTangent=0.0224, tangentMode=RCTM_User, time=1, value=1}
          ]
         }}})
     )
     Y=3.2256
     Z=&Lerp_Float(A=1.8, B=1.2, Alpha=$Particles.NormalizedAge)
     W=0
    )) {isSetParameters=true, module=SetVariables_#17}
  ]
  rendererClasses [&NiagaraSpriteRendererProperties, &NiagaraMeshRendererProperties]
  renderers [
   {rendererClass=&NiagaraSpriteRendererProperties}
   {rendererClass=&NiagaraMeshRendererProperties, rendererIndex=1}
   {rendererClass=&NiagaraSpriteRendererProperties, rendererIndex=2}
   {rendererClass=&NiagaraMeshRendererProperties, rendererIndex=3}
   {rendererClass=&NiagaraMeshRendererProperties, rendererIndex=4}
   {rendererClass=&NiagaraMeshRendererProperties, rendererIndex=5}
   {rendererClass=&NiagaraMeshRendererProperties, rendererIndex=6}
  ]
 }
 Ring001 {
  bEnabled=false
  emitterUpdateScript [
   &EmitterState()
   &SpawnBurst_Instantaneous(Use Spawn Probability=false, UseLoopCountLimit=false)
   &SpawnBurst_Instantaneous(Spawn Group=1, Use Spawn Probability=false, UseLoopCountLimit=false) {module=SpawnBurst_Instantaneous001}
   &SpawnBurst_Instantaneous(Spawn Group=2, Use Spawn Probability=false, UseLoopCountLimit=false) {module=SpawnBurst_Instantaneous002}
  ]
  particleSpawnScript [
   /SetVariables(Particles.VisibilityTag=&SpawnGroup()) {isSetParameters=true, module=SetVariables_#12}
   &InitializeParticle(
    Color=&MultiplyLinearColor(
     Linear Color=$System.SystemTrimCol
     Scale Factor=&SelectVector4_FromArray(
      Array Sampling Mode:ENiagaraArraySamplingMode=::NewEnumerator1
      Direct Array Index:NiagaraInt32=&SpawnGroup()
      Interpolation:NiagaraFloat=?
      Direct Array Mode:ENiagara_WrapClamp=::NewEnumerator0
      Evaluation Type:ENiagaraRandomnessEvaluation=::NewEnumerator0
      Normalized Interpolation:NiagaraBool=true
      Recalculate Random Each Loop:NiagaraBool=false
      Randomness Mode:ENiagaraRandomnessMode=::NewEnumerator0
      Random Seed:NiagaraInt32=?
      Fixed Random Seed:NiagaraBool=false
      Vector4 Selection Array:NiagaraDataInterfaceArrayFloat4=DataInterface() {propertyValues {GpuSyncMode=SyncCpuToGpu, MaxElements=0, FloatData [<2 2 2 0.823>, <2.4 2.4 2.4 10>, <1.15 1.15 1.15 0.6>]}}
     )
    )
    Lifetime=0.26
    Mesh Scale Mode=::NewEnumerator3
    Mesh Uniform Scale=1.2
    Position Offset=<94.539 0 0>
    Sprite Rotation Mode=::NewEnumerator2
    Sprite Size Mode=::NewEnumerator3
    Uniform Sprite Size=&MaskFloatBySpawnGroup(Masked Float=100, Passthrough Float=90)
    UsePositionOffset=true
    Write Material Random=true
   )
   &CameraOffset(Camera Offset Amount=12)
   &InitialMeshOrientation(Orientation Axis=<0 0 1>, Orientation Vector=<0 0 1>, Rotation=<0 0.25 0>, Use Orientation Vector=true, Use Rotation Vector=true)
   &InitialMeshOrientation(
    Orientation Axis=<0 0 0>
    Orientation Vector=<0 0 0>
    Rotation=&RandomRangeVector(Minimum=<0 0 0>, Maximum=<1 0 0>)
    Use Orientation Vector=false
    Use Rotation Vector=true
   ) {module=InitialMeshOrientation001}
   /SetVariables(Particles.WaveParam:Vector4f=<0.3 1 1 1>) {isSetParameters=true, module=SetVariables_#13}
   /SetVariables() {enabled=false, isSetParameters=true, module=SetVariables_#14}
  ]
  particleUpdateScript [
   &ParticleState()
   &ScaleColor(
    Scale Alpha=&FloatFromCurve(FloatCurve=DataInterface() {propertyValues {Curve {keys [
         {arriveTangent=2.8157, interpMode=RCIM_Cubic, leaveTangent=2.8157, tangentMode=RCTM_User, time=0, value=0.2235}
         {arriveTangent=-0.2235, interpMode=RCIM_Cubic, leaveTangent=-0.2235, time=0.0858, value=0.4956}
         {arriveTangent=-0.0522, interpMode=RCIM_Cubic, leaveTangent=-0.0522, tangentMode=RCTM_User, time=1, value=0}
        ]}}})
    ScaleA=true
    ScaleRGB=true
   )
   &ScaleSpriteSize(
    Uniform Curve Scale=$System.FrontendScale
    Uniform Curve Sprite Scale=DataInterface() {propertyValues {Curve {keys [
        {
         arriveTangent=0.7015
         arriveTangentWeight=0.3573
         interpMode=RCIM_Cubic
         leaveTangent=0.7015
         leaveTangentWeight=0.3573
         tangentMode=RCTM_User
         tangentWeightMode=RCTWM_WeightedBoth
         time=0
         value=0.7882
        }
        {
         arriveTangent=0.0045
         arriveTangentWeight=0.7576
         interpMode=RCIM_Cubic
         leaveTangent=0.0045
         leaveTangentWeight=0.7576
         tangentMode=RCTM_User
         tangentWeightMode=RCTWM_WeightedBoth
         time=1
         value=1
        }
       ]}}}
   )
   &ScaleMeshSize(Scale Factor=&VectorFromFloat(Value=&FloatFromCurve(
      FloatCurve=DataInterface() {propertyValues {Curve {
         defaultValue=1
         keys [
          {arriveTangent=7.2689, interpMode=RCIM_Cubic, leaveTangent=7.2689, tangentMode=RCTM_User, time=0, value=0.2235}
          {arriveTangent=0.4035, interpMode=RCIM_Cubic, leaveTangent=0.4035, tangentMode=RCTM_User, time=0.173, value=0.8111}
          {arriveTangent=0.2855, interpMode=RCIM_Cubic, leaveTangent=0.2855, tangentMode=RCTM_User, time=1, value=1}
         ]
        }}}
      Scale Curve=$System.FrontendScale
     )))
   /SetVariables(Particles.RadialColor:LinearColor=&MultiplyLinearColor(Linear Color=$System.SystemTrimCol, Scale Factor=<1 1 1 0.4>)) {enabled=false, isSetParameters=true, module=SetVariables_#15}
   /SetVariables(Particles.RadialParam:Vector4f=&MakeVector4(X=&Lerp_Float(A=-1, B=1.2704, Alpha=$Particles.NormalizedAge), Y=2, Z=3, W=0)) {enabled=false, isSetParameters=true, module=SetVariables_#16}
   /SetVariables(Particles.RadialParam:Vector4f=&MakeVector4(
     X=&Lerp_Float(
      A=-1.5328
      B=-1.2
      Alpha=&FloatFromCurve(FloatCurve=DataInterface() {propertyValues {Curve {
          defaultValue=1
          keys [
           {arriveTangent=2.0823, interpMode=RCIM_Cubic, leaveTangent=2.0823, tangentMode=RCTM_User, time=0, value=0}
           {arriveTangent=0.0224, interpMode=RCIM_Cubic, leaveTangent=0.0224, tangentMode=RCTM_User, time=1, value=1}
          ]
         }}})
     )
     Y=3.2256
     Z=&Lerp_Float(A=1.8, B=1.2, Alpha=$Particles.NormalizedAge)
     W=0
    )) {isSetParameters=true, module=SetVariables_#17}
  ]
  rendererClasses [&NiagaraSpriteRendererProperties, &NiagaraMeshRendererProperties]
  renderers [
   {rendererClass=&NiagaraSpriteRendererProperties}
   {rendererClass=&NiagaraMeshRendererProperties, rendererIndex=1}
   {rendererClass=&NiagaraSpriteRendererProperties, rendererIndex=2}
   {rendererClass=&NiagaraMeshRendererProperties, rendererIndex=3}
   {rendererClass=&NiagaraMeshRendererProperties, rendererIndex=4}
   {rendererClass=&NiagaraMeshRendererProperties, rendererIndex=5}
   {rendererClass=&NiagaraMeshRendererProperties, rendererIndex=6}
  ]
 }
 Sparkle {
  emitterUpdateScript [
   &EmitterState()
   &SpawnBurst_Instantaneous(Spawn Count=9, Use Spawn Probability=false, UseLoopCountLimit=false)
   &SpawnBurst_Instantaneous(Spawn Count=4, Spawn Group=1, Use Spawn Probability=false, UseLoopCountLimit=false) {module=SpawnBurst_Instantaneous001}
   &SpawnBurst_Instantaneous(
    Spawn Count=21
    Spawn Group=2
    Spawn Probability=&MakeCustomFloatFromBool(Bool=$User.bIsHardSurface, True Float=1, False Float=0)
    Spawn Time=0.14
    Use Spawn Probability=true
    UseLoopCountLimit=false
   ) {module=SpawnBurst_Instantaneous002}
   &SpawnBurst_Instantaneous(
    Spawn Count=12
    Spawn Group=3
    Spawn Probability=&MakeCustomFloatFromBool(Bool=$User.bIsHardSurface, True Float=0, False Float=1)
    Use Spawn Probability=true
    UseLoopCountLimit=false
   ) {module=SpawnBurst_Instantaneous003}
  ]
  particleSpawnScript [
   /SetVariables(Particles.VisibilityTag=&MaskIntBySpawnGroup(Masked Int=1, Passthrough Int=0, Spawn Group Mask=3)) {isSetParameters=true}
   &InitializeParticle(
    Color Maximum=$System.SystemCrystalCol2
    Color Minimum=$System.SystemCrystalCol
    Color Mode=::NewEnumerator2
    Lifetime=&Multiply_Float(
     A=&RandomRangeFloat(Minimum=0.6, Maximum=1)
     B=&SelectFloatFromArray(
      Direct Array Index=&SpawnGroup()
      Float Selection Array=DataInterface() {propertyValues {FloatData [1.3,1.3,0.7,0.45]}}
     )
    )
    Sprite Size Mode=::NewEnumerator3
    Uniform Sprite Size=&FloatFromCurve(
     FloatCurve=DataInterface() {propertyValues {Curve {keys [
         <at 0 3 RCIM_Cubic>
         {arriveTangent=3.5882, interpMode=RCIM_Cubic, leaveTangent=3.5882, time=0.8282, value=4.1588}
         {arriveTangent=3.5447, interpMode=RCIM_Cubic, leaveTangent=3.5447, tangentMode=RCTM_User, time=1, value=6.5882}
        ]}}}
     CurveIndex=&RandomRangeFloat(Minimum=0, Maximum=1)
     Scale Curve=$System.FrontEndScale
    )
    UsePositionOffset=false
    Write Material Random=true
   )
   &ShapeLocation(
    MaskSpawnGroup=false
    Override Seed=false
    Sphere Radius=&SelectFloatFromArray(
     Direct Array Index=&SpawnGroup()
     Float Selection Array=DataInterface() {propertyValues {FloatData [12,12,0,12]}}
    )
    UseOwnerScale=true
   )
   &ShapeLocation(
    Hemisphere Distribution=<1 0.65>
    MaskSpawnGroup=false
    Override Seed=false
    Rotation Axis=&Add_Vector(A=$System.DirectionVector, B=<0.5 0 0>)
    Rotation Mode=::NewEnumerator1
    Sphere Radius=&Multiply_Float(
     A=&RandomRangeFloat(Minimum=15, Maximum=45)
     B=&MaskFloatBySpawnGroup(Masked Float=1, Passthrough Float=0, Spawn Group Mask=2)
    )
    Sphere Surface Distribution=1
    UseOwnerScale=true
   ) {module=ShapeLocation001}
   &AddVelocity(
    Velocity=&RandomVector(Vector Scale=1)
    Velocity Speed Scale=&Multiply_Float(
     A=&RandomRangeFloat(Minimum=50, Maximum=90)
     B=&Multiply_Float(
      A=$System.FrontEndScale
      B=&SelectFloatFromArray(
       Direct Array Index=&SpawnGroup()
       Float Selection Array=DataInterface() {propertyValues {FloatData [0.5,1,1,0]}}
      )
     )
    )
   ) {module=AddVelocity001}
   &AddVelocity(
    Velocity=$System.DirectionVector
    Velocity Speed Scale=&Multiply_Float(
     A=&RandomRangeFloat(Minimum=90, Maximum=180)
     B=&Multiply_Float(A=$System.FrontEndScale, B=&MaskFloatBySpawnGroup(Masked Float=1, Passthrough Float=0))
    )
   )
   &AddVelocity(
    Module.Use Falloff Distance=false
    Override Seed=false
    Velocity Mode=::NewEnumerator1
    Velocity Speed=&Multiply_Float(
     A=&RandomRangeFloat(Minimum=100, Maximum=350)
     B=&Multiply_Float(
      A=$System.FrontEndScale
      B=&SelectFloatFromArray(Direct Array Index=&SpawnGroup(), Float Selection Array=DataInterface() {propertyValues {FloatData [0,0,0,1]}})
     )
    )
   ) {module=AddVelocity002}
  ]
  particleUpdateScript [
   &ParticleState()
   &ScaleSpriteSize(Scale Factor=&Vector2DFromFloat(Value=&FloatFromCurve(
      FloatCurve=DataInterface() {propertyValues {Curve {keys [
          <at 0 0 RCIM_Cubic>
          <at 0.1 1>
          <at 0.2 1>
          {arriveTangent=-1.2458, interpMode=RCIM_Cubic, leaveTangent=-1.2458, time=0.3, value=1}
          <at 1.0011 0.002>
         ]}}}
      Scale Curve=&FloatFromCurve(
       FloatCurve=DataInterface() {propertyValues {Curve {keys [<at 0 10>,<at 0.2 2.5>,<at 1 2.125>]}}}
       CurveIndex=&RandomRangeFloat(Minimum=0, Maximum=1, Evaluation Type=::NewEnumerator1)
      )
     )))
   &ScaleColor(
    Scale Alpha=&FloatFromCurve(FloatCurve=DataInterface() {propertyValues {Curve {keys [<at 0 0 RCIM_Cubic>,<at 0.5 1 RCIM_Cubic>,<at 1 0 RCIM_Cubic>]}}})
    ScaleA=true
    ScaleRGB=true
   )
   &CameraOffset(Camera Offset Amount=10)
   &Drag(Drag=&RandomRangeFloat(Minimum=2, Maximum=3), Use Linear Drag=true, Use Rotational Drag=true)
   &GravityForce(Gravity=&VectorFromCurve(
     VectorCurve=DataInterface() {propertyValues {XCurve {keys [<at 0 1>,<at 1 0>]}, YCurve {keys [<at 0 1>,<at 1 0>]}, ZCurve {keys [<at 0.1 0>,<at 0.2 -100>]}}}
     Scale Curve=&SelectVectorFromArray(
      Direct Array Index=&SpawnGroup()
      Vector Selection Array=DataInterface() {propertyValues {FloatData [<1 1 1>, <1 1 1>, <1 1 1>, <1 1 6>]}}
     )
    ))
   &SolveForcesAndVelocity(Clamp Velocity=false, Limit Acceleration=false)
   /SetVariables(
    Particles.Spark Color:LinearColor=&Multiply_LinearColorByFloat(
     Linear Color=$System.SystemTrimCol
     Float=&FloatFromCurve(
      FloatCurve=DataInterface() {propertyValues {Curve {
         defaultValue=0.384
         keys [
          {arriveTangent=9.0725, interpMode=RCIM_Cubic, leaveTangent=9.0725, tangentMode=RCTM_User, time=0, value=0}
          {arriveTangent=-0.0109, interpMode=RCIM_Cubic, leaveTangent=-0.0109, tangentMode=RCTM_User, time=0.1408, value=0.9847}
          {arriveTangent=-0.695, interpMode=RCIM_Cubic, leaveTangent=-0.695, tangentMode=RCTM_User, time=1, value=0}
         ]
        }}}
      Scale Curve=2
     )
    )
    Particles.Spark Size:Vector2f=&Multiply_Vector2D_ByFloat(Float=1, Vector2D=<3 6>)
   ) {isSetParameters=true}
  ]
  rendererClasses [&NiagaraSpriteRendererProperties]
  renderers [
   {rendererClass=&NiagaraSpriteRendererProperties}
   {rendererClass=&NiagaraSpriteRendererProperties, rendererIndex=1}
  ]
 }
 Sparks {
  emitterUpdateScript [
   &EmitterState(Inactive Response=::NewEnumerator1, Life Cycle Mode=::NewEnumerator1, Loop Behavior=::NewEnumerator1, Loop Duration=1, UseLoopDelay=false)
   &SpawnBurst_Instantaneous(
    Spawn Count=&RandomRangeInt(Minimum=15, Maximum=17)
    Spawn Probability=&MakeCustomFloatFromBool(Bool=$User.bIsHardSurface, True Float=1, False Float=0)
    Use Spawn Probability=true
    UseLoopCountLimit=false
   )
   &SpawnBurst_Instantaneous(
    Spawn Count=&RandomRangeInt(Minimum=12, Maximum=14)
    Spawn Group=1
    Spawn Probability=&MakeCustomFloatFromBool(Bool=$User.bIsHardSurface, True Float=1, False Float=0)
    Spawn Time=0.14
    Use Spawn Probability=true
    UseLoopCountLimit=false
   ) {module=SpawnBurst_Instantaneous002}
  ]
  particleSpawnScript [
   /SetVariables(Particles.VisibilityTag=&SpawnGroup()) {isSetParameters=true, module=SetVariables_#18}
   /SetVariables(Particles.DirectionGaussian:NiagaraFloat=&MaskFloatBySpawnGroup(
     Masked Float=&FloatFromCurve(
      FloatCurve=DataInterface() {propertyValues {Curve {
         defaultValue=1
         keys [
          {arriveTangent=1.7876, interpMode=RCIM_Cubic, leaveTangent=1.7876, tangentMode=RCTM_User, time=0, value=0}
          {
           arriveTangentWeight=0.0418
           interpMode=RCIM_Cubic
           leaveTangentWeight=0.0418
           tangentMode=RCTM_User
           tangentWeightMode=RCTWM_WeightedBoth
           time=0.5
           value=0.5
          }
          {arriveTangent=1.7876, interpMode=RCIM_Cubic, leaveTangent=1.7876, tangentMode=RCTM_User, time=1, value=1}
         ]
        }}}
      CurveIndex=&Frac_Float(Float=&Add_Float(A=0.5, B=&Multiply_FloatByInt(Float=1.618, Integer=$Engine.ExecIndex)))
     )
     Passthrough Float=&FloatFromCurve(
      FloatCurve=DataInterface() {propertyValues {Curve {
         defaultValue=1
         keys [
          {arriveTangent=1.7876, interpMode=RCIM_Cubic, leaveTangent=1.7876, tangentMode=RCTM_User, time=0, value=0}
          {
           arriveTangent=-0.0521
           arriveTangentWeight=0.2142
           interpMode=RCIM_Cubic
           leaveTangent=-0.0521
           leaveTangentWeight=0.2142
           tangentMode=RCTM_User
           tangentWeightMode=RCTWM_WeightedBoth
           time=0.5
           value=0.5
          }
          {arriveTangent=1.7876, interpMode=RCIM_Cubic, leaveTangent=1.7876, tangentMode=RCTM_User, time=1, value=1}
         ]
        }}}
      CurveIndex=&Frac_Float(Float=&Multiply_FloatByInt(Float=1.618, Integer=$Engine.ExecIndex))
     )
    )) {isSetParameters=true, module=SetVariables_#19}
   /SetVariables(
    Particles.DirectionRemapped:Vector3f=&RotateVectorByQuaternion(
     Vector To Rotate=$System.DirectionVector
     Quaternion=&MakeQuaternion(
      Quaternion From=0
      Axis=<1 0 0>
      Angle=&FloatFromCurve(
       FloatCurve=DataInterface() {propertyValues {Curve {keys [
           {arriveTangent=0.7228, interpMode=RCIM_Cubic, leaveTangent=0.7228, leaveTangentWeight=0.1667, tangentMode=RCTM_User, time=0, value=-1}
           {
            arriveTangent=3.9993
            arriveTangentWeight=0.5477
            interpMode=RCIM_Cubic
            leaveTangent=3.9993
            leaveTangentWeight=0.5477
            tangentMode=RCTM_User
            tangentWeightMode=RCTWM_WeightedBoth
            time=0.5
            value=0
           }
           {arriveTangent=0.6659, arriveTangentWeight=0.1667, interpMode=RCIM_Cubic, leaveTangent=0.6659, tangentMode=RCTM_User, time=1, value=1}
          ]}}}
       CurveIndex=$Particles.DirectionGaussian
       Scale Curve=180
      )
      From=?
      To=?
      Stabilization Mode:ENiagara_OrientVectorCount=::NewEnumerator0
      Side Direction:Vector3f=?
      Up Direction:Vector3f=?
      Angle Type=::NewEnumerator0
      Normalize:NiagaraBool=true
     )
    )
    Particles.MeshScale2DInitial:Vector2f=&RandomRangeVector2D(Minimum=<1.5 1.04>, Maximum=<4 1.56>)
   ) {isSetParameters=true, module=SetVariables_#20}
   &InitializeParticle(
    Color=&Multiply_LinearColorByFloat(Linear Color=$System.SystemCrystalCol, Float=&RandomRangeFloat(Minimum=0.7, Maximum=1))
    Lifetime=&MaskFloatBySpawnGroup(
     Masked Float=&RandomRangeFloat(Minimum=0.15, Maximum=0.25)
     Passthrough Float=&RandomRangeFloat(Minimum=0.1, Maximum=0.2)
    )
    Mesh Scale=&MaskVectorBySpawnGroup(
     Masked Vector=&Multiply_Vector(
      A=&MakeVector(
       X=&MakeFloatFromVector2d(Vector 2D=$Particles.MeshScale2DInitial, Channel=::NewEnumerator1)
       Y=&MakeFloatFromVector2d(Vector 2D=$Particles.MeshScale2DInitial, Channel=::NewEnumerator1)
       Z=&MakeFloatFromVector2d(Vector 2D=$Particles.MeshScale2DInitial)
      )
      B=&VectorFromCurve(
       VectorCurve=DataInterface() {propertyValues {
         XCurve {
          defaultValue=-1
          keys [
           {arriveTangentWeight=0.4978, interpMode=RCIM_Cubic, leaveTangentWeight=0.4978, time=0, value=1.2007}
           <at 0.5 1.1 RCIM_Cubic>
           <at 1 1.2007 RCIM_Cubic>
          ]
         }
         YCurve {keys [<at 0 1.2007 RCIM_Cubic>,<at 0.5 1.1 RCIM_Cubic>,<at 1 1.2007 RCIM_Cubic>]}
         ZCurve {keys [
           {
            arriveTangentWeight=0.1042
            interpMode=RCIM_Cubic
            leaveTangentWeight=0.1042
            tangentMode=RCTM_User
            tangentWeightMode=RCTWM_WeightedBoth
            time=0.2
            value=0.4
           }
           {
            arriveTangent=25.4578
            arriveTangentWeight=0.5119
            interpMode=RCIM_Cubic
            leaveTangent=-15.9931
            leaveTangentWeight=0.4869
            tangentMode=RCTM_Break
            tangentWeightMode=RCTWM_WeightedBoth
            time=0.5
            value=1.5
           }
           {
            arriveTangentWeight=0.0956
            interpMode=RCIM_Cubic
            leaveTangentWeight=0.0956
            tangentMode=RCTM_User
            tangentWeightMode=RCTWM_WeightedBoth
            time=0.8
            value=0.4
           }
          ]}
        }}
       CurveIndex=$Particles.DirectionGaussian
      )
     )
     Passthrough Vector=&RandomRangeVector(Minimum=<0.42 0.42 0.6>, Maximum=<1 1 1.5>)
    )
    Mesh Scale Mode=::NewEnumerator4
    Position Offset=&MaskVectorBySpawnGroup(
     Masked Vector=<0 0 0>
     Passthrough Vector=&Multiply_VectorByFloat(
      Vector=&Add_Vector(
       A=&Multiply_VectorByFloat(
        Vector=$Particles.DirectionRemapped
        Float=&Add_Float(
         A=&RandomRangeFloat(Minimum=16, Maximum=30)
         B=&FloatFromCurve(
          FloatCurve=DataInterface() {propertyValues {Curve {keys [
              {
               arriveTangentWeight=0.4717
               interpMode=RCIM_Cubic
               leaveTangentWeight=0.4717
               tangentMode=RCTM_User
               tangentWeightMode=RCTWM_WeightedBoth
               time=0
               value=0
              }
              {
               arriveTangent=24.9477
               arriveTangentWeight=0.4945
               interpMode=RCIM_Cubic
               leaveTangent=-14.2663
               leaveTangentWeight=0.4717
               tangentMode=RCTM_Break
               tangentWeightMode=RCTWM_WeightedBoth
               time=0.5
               value=1.2512
              }
              {
               arriveTangentWeight=0.4538
               interpMode=RCIM_Cubic
               leaveTangentWeight=0.4538
               tangentMode=RCTM_User
               tangentWeightMode=RCTWM_WeightedBoth
               time=1
               value=0
              }
             ]}}}
          CurveIndex=$Particles.DirectionGaussian
          Scale Curve=&RandomRangeFloat(Minimum=0, Maximum=40)
         )
        )
       )
       B=<4.6408 0 0>
      )
      Float=$System.FrontendScale
     )
    )
    Position Offset Coordinate Space=::Local
    UsePositionOffset=true
    Write Material Random=true
   )
   &CameraOffset(Camera Offset Amount=15)
   /SetVariables(Particles.MeshIndex=&RandomRangeInt(Minimum=0, Maximum=3)) {isSetParameters=true, module=SetVariables_#21}
   /SetVariables(Particles.OrientationOffset:Vector3f=&Add_Vector(A=$Particles.DirectionRemapped, B=&RandomRangeVector(Minimum=<0.3 0 0>, Maximum=<0.6 0 0>))) {isSetParameters=true, module=SetVariables_#22}
   &InitialMeshOrientation(
    Mesh Axis To Orient=<0 0 1>
    Mesh Orientation Mode=::NewEnumerator2
    Orientation Vector=$Particles.OrientationOffset
    Rotation=&RandomRangeVector(Minimum=<0 0 0>, Maximum=<0 0 1>)
    Use Rotation Vector=true
   )
   /SetVariables(Particles.Vector:Vector3f=&Multiply_Vector(
     A=&MakeVector(
      X=&MakeFloatFromVector2d(Vector 2D=$Particles.MeshScale2DInitial)
      Y=&MakeFloatFromVector2d(Vector 2D=$Particles.MeshScale2DInitial, Channel=::NewEnumerator1)
      Z=&MakeFloatFromVector2d(Vector 2D=$Particles.MeshScale2DInitial, Channel=::NewEnumerator1)
     )
     B=&VectorFromCurve(
      VectorCurve=DataInterface() {propertyValues {
        XCurve {
         defaultValue=-1
         keys [
          {
           arriveTangentWeight=0.4978
           interpMode=RCIM_Cubic
           leaveTangentWeight=0.4978
           tangentMode=RCTM_User
           tangentWeightMode=RCTWM_WeightedBoth
           time=0
           value=0.5129
          }
          {
           arriveTangentWeight=0.0308
           interpMode=RCIM_Cubic
           leaveTangentWeight=0.0308
           tangentMode=RCTM_User
           tangentWeightMode=RCTWM_WeightedBoth
           time=0.5
           value=1
          }
          {
           arriveTangent=0.0277
           arriveTangentWeight=0.4647
           interpMode=RCIM_Cubic
           leaveTangent=0.0277
           leaveTangentWeight=0.4647
           tangentMode=RCTM_User
           tangentWeightMode=RCTWM_WeightedBoth
           time=1
           value=0.5129
          }
         ]
        }
        YCurve {keys [<at 0 1>,<at 1 1>]}
        ZCurve {keys [<at 0 1>,<at 1 1>]}
       }}
      CurveIndex=$Particles.DirectionGaussian
     )
    )) {isSetParameters=true, module=SetVariables_#23}
   &AddVelocity(
    Velocity=$Particles.OrientationOffset
    Velocity Speed Scale=&MaskFloatBySpawnGroup(Masked Float=0, Passthrough Float=&RandomRangeFloat(Minimum=20, Maximum=60))
   )
   &SpriteFacingAndAlignment(Sprite Alignment=$Particles.OrientationOffset, WriteAlignment=true, WriteFacing=false)
  ]
  particleUpdateScript [
   &ParticleState()
   /SetVariables(Particles.MeshScale2D:Vector2f=&MaskVector2dBySpawnGroup(
     Masked Vector2D=&Vector2DFromCurve(
      Vector2Curve=DataInterface() {propertyValues {
        XCurve {keys [<at 0 0>,<at 0.1 1>,<at 0.35 1 RCIM_Constant>,<at 0.85 0>,<at 1 0>,<at 1.1 1>,<at 1.8315 0.9174>,<at 2 0>,<at 2.1 1>,<at 2.8325 1>,<at 2.95 1>,<at 3 0>]}
        YCurve {keys [<at 0 0>,<at 0.15 1.4>,<at 0.35 1 RCIM_Constant>,<at 0.85 0>,<at 1 0>,<at 1.15 1.4>,<at 1.35 1>,<at 1.95 1>,<at 2 0>,<at 2.15 1.4>,<at 2.35 1>,<at 2.8325 1>,<at 3 0>]}
       }}
      CurveIndex=&Add_Float(
       A=$Particles.NormalizedAge
       B=&MakeFloatFromInt(Int=&RandomRangeInt(Minimum=1, Maximum=2))
      )
      ScaleCurve=<1 1>
     )
     Passthrough Vector2D=&Vector2DFromCurve(
      Vector2Curve=DataInterface() {propertyValues {XCurve {keys [<at 0 0.7459>,<at 0.1327 1>,<at 1 0.3049>]}, YCurve {keys [<at 0 0.8721>,<at 0.5 1>,<at 1 0.1672>]}}}
      ScaleCurve=<1 1>
     )
     Spawn Group Mask=0
    )) {isSetParameters=true, module=SetVariables_#24}
   &ScaleMeshSize(Scale Factor=&Multiply_VectorByFloat(
     Vector=&MakeVector(
      X=&MakeFloatFromVector2d(Vector 2D=$Particles.MeshScale2D)
      Y=&MakeFloatFromVector2d(Vector 2D=$Particles.MeshScale2D)
      Z=&MakeFloatFromVector2d(Vector 2D=$Particles.MeshScale2D, Channel=::NewEnumerator1)
     )
     Float=$System.FrontendScale
    ))
   &GravityForce(Gravity=&Multiply_VectorByFloat(
     Vector=<0 0 -200>
     Float=&MaskFloatBySpawnGroup(Masked Float=0, Passthrough Float=&RandomRangeFloat(Minimum=0.5, Maximum=1))
    ))
   &UpdateMeshOrientation(
    Rotation Rate=&MaskFloatBySpawnGroup(Masked Float=0, Passthrough Float=0.4)
    Rotation Vector=&RandomVector(Vector Scale=1)
   )
   &Drag(Drag=3, Rotational Drag=3, Use Linear Drag=true, Use Rotational Drag=true)
   &SolveForcesAndVelocity(Clamp Velocity=false, Limit Acceleration=false)
   /SetVariables(
    Particles.GlowColor=$System.SystemCrystalCol
    Particles.GlowOffset=8
    Particles.GlowParam:Vector4f=<0.2 1 0 0>
    Particles.GlowSize=&Multiply_Vector2D(
     A=&MakeVector2D(
      X=&MakeFloatFromVector(Vector=$Particles.Scale, Channel=::NewEnumerator0)
      Y=&MakeFloatFromVector(Vector=$Particles.Scale, Channel=::NewEnumerator2)
     )
     B=<21.7595 29.7918>
    )
   ) {isSetParameters=true, module=SetVariables_#25}
   /SetVariables(Particles.FlickerColor:LinearColor=/Niagara/DynamicInputs/Add/Add_LinearColor(
     A:LinearColor=&LerpLinearColors(
      StartColor=&Multiply_LinearColorByFloat(Linear Color=$System.SystemCrystalCol, Float=&RandomRangeFloat(Minimum=0.7, Maximum=1))
      EndColor=&Multiply_LinearColorByFloat(
       Linear Color=&HueShiftLinearColor(Hue=0, Saturation=1.4, Value=1.2, Linear Color=$System.SystemTrimCol)
       Float=1
      )
      LerpFactor=&FloatFromCurve(
       FloatCurve=DataInterface() {propertyValues {Curve {keys [
           {arriveTangent=-4.5317, interpMode=RCIM_Cubic, leaveTangent=-4.5317, tangentMode=RCTM_User, time=0.1878, value=1}
           {arriveTangent=-7.9952, tangentMode=RCTM_User, time=0.3089, value=0}
           <at 1 0 RCIM_Cubic>
          ]}}}
       Scale Curve=&MaskFloatBySpawnGroup(Masked Float=&RandomRangeFloat(Minimum=0.2, Maximum=1), Passthrough Float=0)
      )
     )
     B:LinearColor=/Niagara/DynamicInputs/TypeConversions/MakeLinearColorFromFloat(Value:NiagaraFloat=&FloatFromCurve(
       FloatCurve=DataInterface() {propertyValues {Curve {keys [<at 0 0>,<at 0.1754 0.6533>,<at 0.2838 0>,<at 0.345 0.2634>,<at 0.4303 0.149>,<at 0.5361 0.4843>,<at 0.7476 0>,<at 1.2452 0>,<at 1.3259 0.2307>,<at 1.4591 0>,<at 1.5391 0.6635>,<at 1.6457 0>,<at 1.7576 0.4348>,<at 1.8323 0>,<at 2.1468 0>,<at 2.36 0.5083>,<at 2.8611 0>,<at 3 0>]}}}
       CurveIndex=&Add_Float(A=$Particles.NormalizedAge, B=&RandomRangeFloat(Minimum=-0.2, Maximum=2))
       Scale Curve=&MaskFloatBySpawnGroup(
        Masked Float=&RandomRangeFloat(Minimum=0.3, Maximum=1)
        Passthrough Float=&RandomRangeFloat(Minimum=0.3, Maximum=0.6)
       )
      ))
    )) {isSetParameters=true, module=SetVariables_#26}
   /SetVariables(Particles.SpikePosition:NiagaraPosition=&Multiply_VectorByFloat(
     Vector=&Add_Vector(
      A=&Multiply_VectorByFloat(
       Vector=$Particles.DirectionRemapped
       Float=&Add_Float(
        A=&RandomRangeFloat(Minimum=12, Maximum=32)
        B=&FloatFromCurve(
         FloatCurve=DataInterface() {propertyValues {Curve {keys [
             {arriveTangent=1.0181, interpMode=RCIM_Cubic, leaveTangent=1.0181, tangentMode=RCTM_User, time=0, value=-1}
             {arriveTangent=1.296, leaveTangent=1, tangentMode=RCTM_User, time=0.8879, value=0}
             <at 1 0>
            ]}}}
         CurveIndex=&Add_Float(A=$Particles.NormalizedAge, B=&RandomRangeFloat(Minimum=0, Maximum=0.3))
         Scale Curve=&RandomRangeFloat(Minimum=20, Maximum=40)
        )
       )
      )
      B=<4.6408 0 0>
     )
     Float=$System.FrontendScale
    )) {isSetParameters=true, module=SetVariables_#27}
  ]
  rendererClasses [&NiagaraMeshRendererProperties, &NiagaraSpriteRendererProperties]
  renderers [
   {rendererClass=&NiagaraMeshRendererProperties, rendererIndex=0}
   {rendererClass=&NiagaraMeshRendererProperties, rendererIndex=1}
   {rendererClass=&NiagaraSpriteRendererProperties, rendererIndex=2}
   {rendererClass=&NiagaraSpriteRendererProperties, rendererIndex=3}
  ]
 }
 Sparks001 {
  emitterUpdateScript [
   &EmitterState()
   &SpawnBurst_Instantaneous(
    Spawn Count=&RandomRangeInt(Minimum=3, Maximum=5)
    Spawn Probability=&MakeCustomFloatFromBool(Bool=$User.bIsHardSurface, True Float=1, False Float=0)
    Use Spawn Probability=true
    UseLoopCountLimit=false
   )
  ]
  particleSpawnScript [
   &InitializeParticle(
    Color=&Multiply_LinearColorByFloat(Linear Color=$System.SystemCrystalCol, Float=&RandomRangeFloat(Minimum=0.7, Maximum=1))
    Lifetime Max=0.3
    Lifetime Min=0.2
    Lifetime Mode=::NewEnumerator1
    Mesh Scale Max=<0.8 0.8 1.4>
    Mesh Scale Min=<0.6 0.6 1>
    Mesh Scale Mode=::NewEnumerator2
    UsePositionOffset=false
    Write Material Random=true
   )
   &ShapeLocation(
    Hemisphere Distribution=<1 0.14>
    MaskSpawnGroup=false
    Override Seed=false
    Rotation Axis=&Add_Vector(A=$System.DirectionVector, B=<0.5 0 0>)
    Rotation Mode=::NewEnumerator1
    Sphere Radius=&RandomRangeFloat(Minimum=5, Maximum=15)
    Sphere Surface Distribution=1
    UseOwnerScale=true
   )
   &CameraOffset(Camera Offset Amount=40)
   &AddVelocity(
    Module.Use Falloff Distance=false
    Override Seed=false
    Velocity Mode=::NewEnumerator1
    Velocity Speed=&RandomRangeFloat(Minimum=300, Maximum=700)
   )
   /SetVariables(Particles.MeshIndex=&RandomRangeInt(Minimum=0, Maximum=3)) {isSetParameters=true, module=SetVariables_#21}
   &InitialMeshOrientation(Mesh Orientation Mode=::NewEnumerator1, Rotation=<0 0.25 0>, Use Rotation Vector=true) {module=InitialMeshOrientation001}
  ]
  particleUpdateScript [
   &ParticleState()
   &Drag(
    Drag=&RandomRangeFloat(Minimum=2, Maximum=3.5)
    Rotational Drag=5
    Use Linear Drag=true
    Use Rotational Drag=true
   )
   &ScaleMeshSize(Scale Factor=&VectorFromCurve(
     VectorCurve=DataInterface() {propertyValues {
       XCurve {keys [<at 0.75 1>,<at 1 0.2237>]}
       YCurve {keys [<at 0.75 1>,<at 1 0.2237>]}
       ZCurve {keys [<at 0 0>,<at 0.1976 1>,<at 0.907 1>,<at 1 0>]}
      }}
     Scale Curve=&VectorFromFloat(Value=$System.FrontendScale)
    ))
   &GravityForce(Gravity=&Multiply_VectorByFloat(Vector=<0 0 -400>, Float=&RandomRangeFloat(Minimum=0.5, Maximum=1)))
   &SolveForcesAndVelocity(Clamp Velocity=false, Limit Acceleration=false)
   /SetVariables(
    Particles.GlowColor=$System.SystemCrystalCol
    Particles.GlowOffset=8
    Particles.GlowParam:Vector4f=<0.2 1 0 0>
    Particles.GlowSize=&Multiply_Vector2D(
     A=&MakeVector2D(
      X=&MakeFloatFromVector(Vector=$Particles.Scale, Channel=::NewEnumerator0)
      Y=&MakeFloatFromVector(Vector=$Particles.Scale, Channel=::NewEnumerator2)
     )
     B=<40 40>
    )
   ) {isSetParameters=true}
  ]
  rendererClasses [&NiagaraMeshRendererProperties, &NiagaraSpriteRendererProperties]
  renderers [
   {rendererClass=&NiagaraMeshRendererProperties, rendererIndex=0}
   {rendererClass=&NiagaraSpriteRendererProperties, rendererIndex=1}
  ]
 }
 Sparks002 {
  bEnabled=false
  emitterUpdateScript [
   &EmitterState(Inactive Response=::NewEnumerator1, Life Cycle Mode=::NewEnumerator1, Loop Behavior=::NewEnumerator1, Loop Duration=1, UseLoopDelay=false)
   &SpawnBurst_Instantaneous(
    Spawn Count=&RandomRangeInt(Minimum=15, Maximum=17)
    Spawn Probability=&MakeCustomFloatFromBool(Bool=$User.bIsHardSurface, True Float=1, False Float=0)
    Use Spawn Probability=true
    UseLoopCountLimit=false
   )
   &SpawnBurst_Instantaneous(
    Spawn Count=&RandomRangeInt(Minimum=19, Maximum=25)
    Spawn Group=1
    Spawn Probability=&MakeCustomFloatFromBool(Bool=$User.bIsHardSurface, True Float=1, False Float=0)
    Spawn Time=0.16
    Use Spawn Probability=true
    UseLoopCountLimit=false
   ) {module=SpawnBurst_Instantaneous002}
  ]
  particleSpawnScript [
   /SetVariables(Particles.VisibilityTag=&SpawnGroup()) {isSetParameters=true, module=SetVariables_#18}
   /SetVariables(Particles.DirectionGaussian:NiagaraFloat=&Lerp_Float(
     A=0
     B=1
     Alpha=&MaskFloatBySpawnGroup(
      Masked Float=&FloatFromCurve(
       FloatCurve=DataInterface() {propertyValues {Curve {
          defaultValue=1
          keys [
           {arriveTangent=1.7876, interpMode=RCIM_Cubic, leaveTangent=1.7876, tangentMode=RCTM_User, time=0, value=0}
           {
            arriveTangentWeight=0.0418
            interpMode=RCIM_Cubic
            leaveTangentWeight=0.0418
            tangentMode=RCTM_User
            tangentWeightMode=RCTWM_WeightedBoth
            time=0.5
            value=0.5
           }
           {arriveTangent=1.7876, interpMode=RCIM_Cubic, leaveTangent=1.7876, tangentMode=RCTM_User, time=1, value=1}
          ]
         }}}
       CurveIndex=&Frac_Float(Float=&Add_Float(A=0.5, B=&Multiply_FloatByInt(Float=1.618, Integer=$Engine.ExecIndex)))
      )
      Passthrough Float=&FloatFromCurve(
       FloatCurve=DataInterface() {propertyValues {Curve {
          defaultValue=1
          keys [
           {arriveTangent=1.7876, interpMode=RCIM_Cubic, leaveTangent=1.7876, tangentMode=RCTM_User, time=0, value=0}
           {
            arriveTangentWeight=0.0418
            interpMode=RCIM_Cubic
            leaveTangentWeight=0.0418
            tangentMode=RCTM_User
            tangentWeightMode=RCTWM_WeightedBoth
            time=0.5
            value=0.5
           }
           {arriveTangent=1.7876, interpMode=RCIM_Cubic, leaveTangent=1.7876, tangentMode=RCTM_User, time=1, value=1}
          ]
         }}}
       CurveIndex=&RandomRangeFloat(Minimum=0, Maximum=1)
      )
     )
    )) {isSetParameters=true, module=SetVariables_#19}
   /SetVariables(
    Particles.DirectionRemapped:Vector3f=&RotateVectorByQuaternion(
     Vector To Rotate=$System.DirectionVector
     Quaternion=&MakeQuaternion(
      Quaternion From=0
      Axis=<1 0 0>
      Angle=&FloatFromCurve(
       FloatCurve=DataInterface() {propertyValues {Curve {keys [
           {arriveTangent=0.7228, interpMode=RCIM_Cubic, leaveTangent=0.7228, leaveTangentWeight=0.1667, tangentMode=RCTM_User, time=0, value=-1}
           {
            arriveTangent=3.9993
            arriveTangentWeight=0.5477
            interpMode=RCIM_Cubic
            leaveTangent=3.9993
            leaveTangentWeight=0.5477
            tangentMode=RCTM_User
            tangentWeightMode=RCTWM_WeightedBoth
            time=0.5
            value=0
           }
           {arriveTangent=0.6659, arriveTangentWeight=0.1667, interpMode=RCIM_Cubic, leaveTangent=0.6659, tangentMode=RCTM_User, time=1, value=1}
          ]}}}
       CurveIndex=$Particles.DirectionGaussian
       Scale Curve=180
      )
      From=?
      To=?
      Stabilization Mode:ENiagara_OrientVectorCount=::NewEnumerator0
      Side Direction:Vector3f=?
      Up Direction:Vector3f=?
      Angle Type=::NewEnumerator0
      Normalize:NiagaraBool=true
     )
    )
    Particles.MeshScale2DInitial:Vector2f=&RandomRangeVector2D(Minimum=<2 0.8>, Maximum=<4 1.2>)
   ) {isSetParameters=true, module=SetVariables_#20}
   &InitializeParticle(
    Color=&Multiply_LinearColorByFloat(Linear Color=$System.SystemCrystalCol, Float=&RandomRangeFloat(Minimum=0.7, Maximum=1))
    Lifetime=&MaskFloatBySpawnGroup(
     Masked Float=&RandomRangeFloat(Minimum=0.15, Maximum=0.25)
     Passthrough Float=&RandomRangeFloat(Minimum=0.1, Maximum=0.2)
    )
    Mesh Scale=&MaskVectorBySpawnGroup(
     Masked Vector=&Multiply_Vector(
      A=&MakeVector(
       X=&MakeFloatFromVector2d(Vector 2D=$Particles.MeshScale2DInitial, Channel=::NewEnumerator1)
       Y=&MakeFloatFromVector2d(Vector 2D=$Particles.MeshScale2DInitial, Channel=::NewEnumerator1)
       Z=&MakeFloatFromVector2d(Vector 2D=$Particles.MeshScale2DInitial)
      )
      B=&VectorFromCurve(
       VectorCurve=DataInterface() {propertyValues {
         XCurve {
          defaultValue=-1
          keys [
           {arriveTangentWeight=0.4978, interpMode=RCIM_Cubic, leaveTangentWeight=0.4978, time=0, value=1}
           <at 0.5 1.2 RCIM_Cubic>
           <at 1 1 RCIM_Cubic>
          ]
         }
         YCurve {keys [<at 0 1 RCIM_Cubic>,<at 0.5 1.2 RCIM_Cubic>,<at 1 1 RCIM_Cubic>]}
         ZCurve {keys [
           {
            arriveTangentWeight=0.1042
            interpMode=RCIM_Cubic
            leaveTangentWeight=0.1042
            tangentMode=RCTM_User
            tangentWeightMode=RCTWM_WeightedBoth
            time=0.2
            value=0.6987
           }
           {
            arriveTangent=25.4578
            arriveTangentWeight=0.5119
            interpMode=RCIM_Cubic
            leaveTangent=-15.9931
            leaveTangentWeight=0.4869
            tangentMode=RCTM_Break
            tangentWeightMode=RCTWM_WeightedBoth
            time=0.5
            value=1.5
           }
           {
            arriveTangentWeight=0.0956
            interpMode=RCIM_Cubic
            leaveTangentWeight=0.0956
            tangentMode=RCTM_User
            tangentWeightMode=RCTWM_WeightedBoth
            time=0.8
            value=0.6987
           }
          ]}
        }}
       CurveIndex=$Particles.DirectionGaussian
      )
     )
     Passthrough Vector=&RandomRangeVector(Minimum=<0.42 0.42 0.6>, Maximum=<1 1 1.5>)
    )
    Mesh Scale Mode=::NewEnumerator4
    Position Offset=&Add_Vector(
     A=&Multiply_VectorByFloat(
      Vector=$Particles.DirectionRemapped
      Float=&MaskFloatBySpawnGroup(
       Masked Float=&RandomRangeFloat(Minimum=7.5, Maximum=20)
       Passthrough Float=&Add_Float(
        A=&RandomRangeFloat(Minimum=8, Maximum=26)
        B=&FloatFromCurve(
         FloatCurve=DataInterface() {propertyValues {Curve {keys [
             {
              arriveTangentWeight=0.4717
              interpMode=RCIM_Cubic
              leaveTangentWeight=0.4717
              tangentMode=RCTM_User
              tangentWeightMode=RCTWM_WeightedBoth
              time=0
              value=0
             }
             {
              arriveTangent=24.9477
              arriveTangentWeight=0.4945
              interpMode=RCIM_Cubic
              leaveTangent=-14.2663
              leaveTangentWeight=0.4717
              tangentMode=RCTM_Break
              tangentWeightMode=RCTWM_WeightedBoth
              time=0.5
              value=1.2512
             }
             {
              arriveTangentWeight=0.4538
              interpMode=RCIM_Cubic
              leaveTangentWeight=0.4538
              tangentMode=RCTM_User
              tangentWeightMode=RCTWM_WeightedBoth
              time=1
              value=0
             }
            ]}}}
         CurveIndex=$Particles.DirectionGaussian
         Scale Curve=&RandomRangeFloat(Minimum=0, Maximum=40)
        )
       )
      )
     )
     B=<4.6408 0 0>
    )
    Position Offset Coordinate Space=::Local
    UsePositionOffset=true
    Write Material Random=true
   )
   &CameraOffset(Camera Offset Amount=15)
   /SetVariables(Particles.MeshIndex=&RandomRangeInt(Minimum=0, Maximum=3)) {isSetParameters=true, module=SetVariables_#21}
   /SetVariables(Particles.OrientationOffset:Vector3f=&Add_Vector(A=$Particles.DirectionRemapped, B=&RandomRangeVector(Minimum=<0.3 0 0>, Maximum=<0.6 0 0>))) {isSetParameters=true, module=SetVariables_#22}
   &InitialMeshOrientation(
    Mesh Axis To Orient=<0 0 1>
    Mesh Orientation Mode=::NewEnumerator2
    Orientation Vector=$Particles.OrientationOffset
    Rotation=&RandomRangeVector(Minimum=<0 0 0>, Maximum=<0 0 1>)
    Use Rotation Vector=true
   )
   /SetVariables(Particles.Vector:Vector3f=&Multiply_Vector(
     A=&MakeVector(
      X=&MakeFloatFromVector2d(Vector 2D=$Particles.MeshScale2DInitial)
      Y=&MakeFloatFromVector2d(Vector 2D=$Particles.MeshScale2DInitial, Channel=::NewEnumerator1)
      Z=&MakeFloatFromVector2d(Vector 2D=$Particles.MeshScale2DInitial, Channel=::NewEnumerator1)
     )
     B=&VectorFromCurve(
      VectorCurve=DataInterface() {propertyValues {
        XCurve {
         defaultValue=-1
         keys [
          {
           arriveTangentWeight=0.4978
           interpMode=RCIM_Cubic
           leaveTangentWeight=0.4978
           tangentMode=RCTM_User
           tangentWeightMode=RCTWM_WeightedBoth
           time=0
           value=0.5129
          }
          {
           arriveTangentWeight=0.0308
           interpMode=RCIM_Cubic
           leaveTangentWeight=0.0308
           tangentMode=RCTM_User
           tangentWeightMode=RCTWM_WeightedBoth
           time=0.5
           value=1
          }
          {
           arriveTangent=0.0277
           arriveTangentWeight=0.4647
           interpMode=RCIM_Cubic
           leaveTangent=0.0277
           leaveTangentWeight=0.4647
           tangentMode=RCTM_User
           tangentWeightMode=RCTWM_WeightedBoth
           time=1
           value=0.5129
          }
         ]
        }
        YCurve {keys [<at 0 1>,<at 1 1>]}
        ZCurve {keys [<at 0 1>,<at 1 1>]}
       }}
      CurveIndex=$Particles.DirectionGaussian
     )
    )) {isSetParameters=true, module=SetVariables_#23}
   &AddVelocity(
    Velocity=$Particles.OrientationOffset
    Velocity Speed Scale=&MaskFloatBySpawnGroup(Masked Float=0, Passthrough Float=&RandomRangeFloat(Minimum=20, Maximum=60))
   )
   &SpriteFacingAndAlignment(Sprite Alignment=$Particles.OrientationOffset, WriteAlignment=true, WriteFacing=false)
  ]
  particleUpdateScript [
   &ParticleState()
   /SetVariables(Particles.MeshScale2D:Vector2f=&MaskVector2dBySpawnGroup(
     Masked Vector2D=&Vector2DFromCurve(
      Vector2Curve=DataInterface() {propertyValues {
        XCurve {keys [<at 0 0>,<at 0.1 1>,<at 0.35 1 RCIM_Constant>,<at 0.85 0>,<at 1 0>,<at 1.1 1>,<at 1.8315 0.9174>,<at 2 0>,<at 2.1 1>,<at 2.8325 1>,<at 2.95 1>,<at 3 0>]}
        YCurve {keys [<at 0 0>,<at 0.15 1.4>,<at 0.35 1 RCIM_Constant>,<at 0.85 0>,<at 1 0>,<at 1.15 1.4>,<at 1.35 1>,<at 1.95 1>,<at 2 0>,<at 2.15 1.4>,<at 2.35 1>,<at 2.8325 1>,<at 3 0>]}
       }}
      CurveIndex=&Add_Float(
       A=$Particles.NormalizedAge
       B=&MakeFloatFromInt(Int=&RandomRangeInt(Minimum=1, Maximum=1))
      )
      ScaleCurve=<1 1>
     )
     Passthrough Vector2D=&Vector2DFromCurve(
      Vector2Curve=DataInterface() {propertyValues {XCurve {keys [<at 0 0.7459>,<at 0.1327 1>,<at 1 0.3049>]}, YCurve {keys [<at 0 0.8721>,<at 0.5 1>,<at 1 0.1672>]}}}
      ScaleCurve=<1 1>
     )
     Spawn Group Mask=0
    )) {isSetParameters=true, module=SetVariables_#24}
   &ScaleMeshSize(Scale Factor=&MakeVector(
     X=&MakeFloatFromVector2d(Vector 2D=$Particles.MeshScale2D)
     Y=&MakeFloatFromVector2d(Vector 2D=$Particles.MeshScale2D)
     Z=&MakeFloatFromVector2d(Vector 2D=$Particles.MeshScale2D, Channel=::NewEnumerator1)
    ))
   &GravityForce(Gravity=&Multiply_VectorByFloat(
     Vector=<0 0 -200>
     Float=&MaskFloatBySpawnGroup(Masked Float=0, Passthrough Float=&RandomRangeFloat(Minimum=0.5, Maximum=1))
    ))
   &UpdateMeshOrientation(
    Rotation Rate=&MaskFloatBySpawnGroup(Masked Float=0, Passthrough Float=0.4)
    Rotation Vector=&RandomVector(Vector Scale=1)
   )
   &Drag(Drag=3, Rotational Drag=3, Use Linear Drag=true, Use Rotational Drag=true)
   &SolveForcesAndVelocity(Clamp Velocity=false, Limit Acceleration=false)
   /SetVariables(
    Particles.GlowColor=$System.SystemCrystalCol
    Particles.GlowOffset=8
    Particles.GlowParam:Vector4f=<0.2 1 0 0>
    Particles.GlowSize=&Multiply_Vector2D(
     A=&MakeVector2D(
      X=&MakeFloatFromVector(Vector=$Particles.Scale, Channel=::NewEnumerator0)
      Y=&MakeFloatFromVector(Vector=$Particles.Scale, Channel=::NewEnumerator2)
     )
     B=<21.7595 29.7918>
    )
   ) {isSetParameters=true, module=SetVariables_#25}
   /SetVariables(Particles.FlickerColor:LinearColor=&Add_LinearColor(
     A:LinearColor=&Multiply_LinearColorByFloat(Linear Color=$System.SystemCrystalCol, Float=&RandomRangeFloat(Minimum=0.7, Maximum=1))
     B:LinearColor=&MakeLinearColorFromFloat(Value:NiagaraFloat=&FloatFromCurve(
       FloatCurve=DataInterface() {propertyValues {Curve {keys [<at 0 0>,<at 0.1754 0.6533>,<at 0.2838 0>,<at 0.345 0.2634>,<at 0.4303 0.149>,<at 0.5361 0.4843>,<at 0.7476 0>,<at 1.2452 0>,<at 1.3259 0.2307>,<at 1.4591 0>,<at 1.5391 0.6635>,<at 1.6457 0>,<at 1.7576 0.4348>,<at 1.8323 0>,<at 2.1468 0>,<at 2.36 0.5083>,<at 2.8611 0>,<at 3 0>]}}}
       CurveIndex=&Add_Float(A=$Particles.NormalizedAge, B=&RandomRangeFloat(Minimum=-0.5, Maximum=2))
       Scale Curve=&RandomRangeFloat(Minimum=0, Maximum=1)
      ))
    )) {isSetParameters=true, module=SetVariables_#26}
   /SetVariables(Particles.SpikePosition:NiagaraPosition=&Add_Vector(
     A=&Multiply_VectorByFloat(
      Vector=$Particles.DirectionRemapped
      Float=&Add_Float(
       A=&RandomRangeFloat(Minimum=9, Maximum=26)
       B=&FloatFromCurve(
        FloatCurve=DataInterface() {propertyValues {Curve {keys [
            {arriveTangent=2.2394, interpMode=RCIM_Cubic, leaveTangent=2.2394, tangentMode=RCTM_User, time=0, value=-1}
            {arriveTangent=0.1323, leaveTangent=1, tangentMode=RCTM_User, time=0.8879, value=0}
            <at 1 0>
           ]}}}
        CurveIndex=&Add_Float(A=$Particles.NormalizedAge, B=&RandomRangeFloat(Minimum=0, Maximum=0.3))
        Scale Curve=&RandomRangeFloat(Minimum=20, Maximum=40)
       )
      )
     )
     B=<4.6408 0 0>
    )) {isSetParameters=true, module=SetVariables_#27}
  ]
  rendererClasses [&NiagaraMeshRendererProperties, &NiagaraSpriteRendererProperties]
  renderers [
   {rendererClass=&NiagaraMeshRendererProperties, rendererIndex=0}
   {rendererClass=&NiagaraMeshRendererProperties, rendererIndex=1}
   {rendererClass=&NiagaraSpriteRendererProperties, rendererIndex=2}
   {rendererClass=&NiagaraSpriteRendererProperties, rendererIndex=3}
  ]
 }
 Sparks003 {
  bEnabled=false
  emitterUpdateScript [
   &EmitterState(Inactive Response=::NewEnumerator1, Life Cycle Mode=::NewEnumerator1, Loop Behavior=::NewEnumerator1, Loop Duration=1, UseLoopDelay=false)
   &SpawnBurst_Instantaneous(
    Spawn Count=&RandomRangeInt(Minimum=15, Maximum=17)
    Spawn Probability=&MakeCustomFloatFromBool(Bool=$User.bIsHardSurface, True Float=1, False Float=0)
    Use Spawn Probability=true
    UseLoopCountLimit=false
   )
   &SpawnBurst_Instantaneous(
    Spawn Count=&RandomRangeInt(Minimum=19, Maximum=25)
    Spawn Group=1
    Spawn Probability=&MakeCustomFloatFromBool(Bool=$User.bIsHardSurface, True Float=1, False Float=0)
    Spawn Time=0.16
    Use Spawn Probability=true
    UseLoopCountLimit=false
   ) {module=SpawnBurst_Instantaneous002}
  ]
  particleSpawnScript [
   /SetVariables(Particles.VisibilityTag=&SpawnGroup()) {isSetParameters=true, module=SetVariables_#18}
   /SetVariables(Particles.DirectionGaussian:NiagaraFloat=&MaskFloatBySpawnGroup(
     Masked Float=&FloatFromCurve(
      FloatCurve=DataInterface() {propertyValues {Curve {
         defaultValue=1
         keys [
          {arriveTangent=1.7876, interpMode=RCIM_Cubic, leaveTangent=1.7876, tangentMode=RCTM_User, time=0, value=0}
          {
           arriveTangentWeight=0.0418
           interpMode=RCIM_Cubic
           leaveTangentWeight=0.0418
           tangentMode=RCTM_User
           tangentWeightMode=RCTWM_WeightedBoth
           time=0.5
           value=0.5
          }
          {arriveTangent=1.7876, interpMode=RCIM_Cubic, leaveTangent=1.7876, tangentMode=RCTM_User, time=1, value=1}
         ]
        }}}
      CurveIndex=&Frac_Float(Float=&Add_Float(A=0.5, B=&Multiply_FloatByInt(Float=1.618, Integer=$Engine.ExecIndex)))
     )
     Passthrough Float=&FloatFromCurve(
      FloatCurve=DataInterface() {propertyValues {Curve {
         defaultValue=1
         keys [
          {arriveTangent=1.7876, interpMode=RCIM_Cubic, leaveTangent=1.7876, tangentMode=RCTM_User, time=0, value=0}
          {
           arriveTangentWeight=0.0418
           interpMode=RCIM_Cubic
           leaveTangentWeight=0.0418
           tangentMode=RCTM_User
           tangentWeightMode=RCTWM_WeightedBoth
           time=0.5
           value=0.5
          }
          {arriveTangent=1.7876, interpMode=RCIM_Cubic, leaveTangent=1.7876, tangentMode=RCTM_User, time=1, value=1}
         ]
        }}}
      CurveIndex=&RandomRangeFloat(Minimum=0, Maximum=1)
     )
    )) {isSetParameters=true, module=SetVariables_#19}
   /SetVariables(
    Particles.DirectionRemapped:Vector3f=&RotateVectorByQuaternion(
     Vector To Rotate=$System.DirectionVector
     Quaternion=&MakeQuaternion(
      Quaternion From=0
      Axis=<1 0 0>
      Angle=&FloatFromCurve(
       FloatCurve=DataInterface() {propertyValues {Curve {keys [
           {arriveTangent=0.7228, interpMode=RCIM_Cubic, leaveTangent=0.7228, leaveTangentWeight=0.1667, tangentMode=RCTM_User, time=0, value=-1}
           {
            arriveTangent=3.9993
            arriveTangentWeight=0.5477
            interpMode=RCIM_Cubic
            leaveTangent=3.9993
            leaveTangentWeight=0.5477
            tangentMode=RCTM_User
            tangentWeightMode=RCTWM_WeightedBoth
            time=0.5
            value=0
           }
           {arriveTangent=0.6659, arriveTangentWeight=0.1667, interpMode=RCIM_Cubic, leaveTangent=0.6659, tangentMode=RCTM_User, time=1, value=1}
          ]}}}
       CurveIndex=$Particles.DirectionGaussian
       Scale Curve=180
      )
      From=?
      To=?
      Stabilization Mode:ENiagara_OrientVectorCount=::NewEnumerator0
      Side Direction:Vector3f=?
      Up Direction:Vector3f=?
      Angle Type=::NewEnumerator0
      Normalize:NiagaraBool=true
     )
    )
    Particles.MeshScale2DInitial:Vector2f=&RandomRangeVector2D(Minimum=<1.5 1.04>, Maximum=<4 1.56>)
   ) {isSetParameters=true, module=SetVariables_#20}
   &InitializeParticle(
    Color=&Multiply_LinearColorByFloat(Linear Color=$System.SystemCrystalCol, Float=&RandomRangeFloat(Minimum=0.7, Maximum=1))
    Lifetime=&MaskFloatBySpawnGroup(
     Masked Float=&RandomRangeFloat(Minimum=0.15, Maximum=0.25)
     Passthrough Float=&RandomRangeFloat(Minimum=0.1, Maximum=0.2)
    )
    Mesh Scale=&MaskVectorBySpawnGroup(
     Masked Vector=&Multiply_Vector(
      A=&MakeVector(
       X=&MakeFloatFromVector2d(Vector 2D=$Particles.MeshScale2DInitial, Channel=::NewEnumerator1)
       Y=&MakeFloatFromVector2d(Vector 2D=$Particles.MeshScale2DInitial, Channel=::NewEnumerator1)
       Z=&MakeFloatFromVector2d(Vector 2D=$Particles.MeshScale2DInitial)
      )
      B=&VectorFromCurve(
       VectorCurve=DataInterface() {propertyValues {
         XCurve {
          defaultValue=-1
          keys [
           {arriveTangentWeight=0.4978, interpMode=RCIM_Cubic, leaveTangentWeight=0.4978, time=0, value=1.2007}
           <at 0.5 1.1 RCIM_Cubic>
           <at 1 1.2007 RCIM_Cubic>
          ]
         }
         YCurve {keys [<at 0 1.2007 RCIM_Cubic>,<at 0.5 1.1 RCIM_Cubic>,<at 1 1.2007 RCIM_Cubic>]}
         ZCurve {keys [
           {
            arriveTangentWeight=0.1042
            interpMode=RCIM_Cubic
            leaveTangentWeight=0.1042
            tangentMode=RCTM_User
            tangentWeightMode=RCTWM_WeightedBoth
            time=0.2
            value=0.4
           }
           {
            arriveTangent=25.4578
            arriveTangentWeight=0.5119
            interpMode=RCIM_Cubic
            leaveTangent=-15.9931
            leaveTangentWeight=0.4869
            tangentMode=RCTM_Break
            tangentWeightMode=RCTWM_WeightedBoth
            time=0.5
            value=1.5
           }
           {
            arriveTangentWeight=0.0956
            interpMode=RCIM_Cubic
            leaveTangentWeight=0.0956
            tangentMode=RCTM_User
            tangentWeightMode=RCTWM_WeightedBoth
            time=0.8
            value=0.4
           }
          ]}
        }}
       CurveIndex=$Particles.DirectionGaussian
      )
     )
     Passthrough Vector=&RandomRangeVector(Minimum=<0.42 0.42 0.6>, Maximum=<1 1 1.5>)
    )
    Mesh Scale Mode=::NewEnumerator4
    Position Offset=&Add_Vector(
     A=&Multiply_VectorByFloat(
      Vector=$Particles.DirectionRemapped
      Float=&MaskFloatBySpawnGroup(
       Masked Float=0
       Passthrough Float=&Add_Float(
        A=&RandomRangeFloat(Minimum=8, Maximum=26)
        B=&FloatFromCurve(
         FloatCurve=DataInterface() {propertyValues {Curve {keys [
             {
              arriveTangentWeight=0.4717
              interpMode=RCIM_Cubic
              leaveTangentWeight=0.4717
              tangentMode=RCTM_User
              tangentWeightMode=RCTWM_WeightedBoth
              time=0
              value=0
             }
             {
              arriveTangent=24.9477
              arriveTangentWeight=0.4945
              interpMode=RCIM_Cubic
              leaveTangent=-14.2663
              leaveTangentWeight=0.4717
              tangentMode=RCTM_Break
              tangentWeightMode=RCTWM_WeightedBoth
              time=0.5
              value=1.2512
             }
             {
              arriveTangentWeight=0.4538
              interpMode=RCIM_Cubic
              leaveTangentWeight=0.4538
              tangentMode=RCTM_User
              tangentWeightMode=RCTWM_WeightedBoth
              time=1
              value=0
             }
            ]}}}
         CurveIndex=$Particles.DirectionGaussian
         Scale Curve=&RandomRangeFloat(Minimum=0, Maximum=40)
        )
       )
      )
     )
     B=<4.6408 0 0>
    )
    Position Offset Coordinate Space=::Local
    UsePositionOffset=true
    Write Material Random=true
   )
   &CameraOffset(Camera Offset Amount=15)
   /SetVariables(Particles.MeshIndex=&RandomRangeInt(Minimum=0, Maximum=3)) {isSetParameters=true, module=SetVariables_#21}
   /SetVariables(Particles.OrientationOffset:Vector3f=&Add_Vector(A=$Particles.DirectionRemapped, B=&RandomRangeVector(Minimum=<0.3 0 0>, Maximum=<0.6 0 0>))) {isSetParameters=true, module=SetVariables_#22}
   &InitialMeshOrientation(
    Mesh Axis To Orient=<0 0 1>
    Mesh Orientation Mode=::NewEnumerator2
    Orientation Vector=$Particles.OrientationOffset
    Rotation=&RandomRangeVector(Minimum=<0 0 0>, Maximum=<0 0 1>)
    Use Rotation Vector=true
   )
   /SetVariables(Particles.Vector:Vector3f=&Multiply_Vector(
     A=&MakeVector(
      X=&MakeFloatFromVector2d(Vector 2D=$Particles.MeshScale2DInitial)
      Y=&MakeFloatFromVector2d(Vector 2D=$Particles.MeshScale2DInitial, Channel=::NewEnumerator1)
      Z=&MakeFloatFromVector2d(Vector 2D=$Particles.MeshScale2DInitial, Channel=::NewEnumerator1)
     )
     B=&VectorFromCurve(
      VectorCurve=DataInterface() {propertyValues {
        XCurve {
         defaultValue=-1
         keys [
          {
           arriveTangentWeight=0.4978
           interpMode=RCIM_Cubic
           leaveTangentWeight=0.4978
           tangentMode=RCTM_User
           tangentWeightMode=RCTWM_WeightedBoth
           time=0
           value=0.5129
          }
          {
           arriveTangentWeight=0.0308
           interpMode=RCIM_Cubic
           leaveTangentWeight=0.0308
           tangentMode=RCTM_User
           tangentWeightMode=RCTWM_WeightedBoth
           time=0.5
           value=1
          }
          {
           arriveTangent=0.0277
           arriveTangentWeight=0.4647
           interpMode=RCIM_Cubic
           leaveTangent=0.0277
           leaveTangentWeight=0.4647
           tangentMode=RCTM_User
           tangentWeightMode=RCTWM_WeightedBoth
           time=1
           value=0.5129
          }
         ]
        }
        YCurve {keys [<at 0 1>,<at 1 1>]}
        ZCurve {keys [<at 0 1>,<at 1 1>]}
       }}
      CurveIndex=$Particles.DirectionGaussian
     )
    )) {isSetParameters=true, module=SetVariables_#23}
   &AddVelocity(
    Velocity=$Particles.OrientationOffset
    Velocity Speed Scale=&MaskFloatBySpawnGroup(Masked Float=0, Passthrough Float=&RandomRangeFloat(Minimum=20, Maximum=60))
   )
   &SpriteFacingAndAlignment(Sprite Alignment=$Particles.OrientationOffset, WriteAlignment=true, WriteFacing=false)
  ]
  particleUpdateScript [
   &ParticleState()
   /SetVariables(Particles.MeshScale2D:Vector2f=&MaskVector2dBySpawnGroup(
     Masked Vector2D=&Vector2DFromCurve(
      Vector2Curve=DataInterface() {propertyValues {
        XCurve {keys [<at 0 0>,<at 0.1 1>,<at 0.35 1 RCIM_Constant>,<at 0.85 0>,<at 1 0>,<at 1.1 1>,<at 1.8315 0.9174>,<at 2 0>,<at 2.1 1>,<at 2.8325 1>,<at 2.95 1>,<at 3 0>]}
        YCurve {keys [<at 0 0>,<at 0.15 1.4>,<at 0.35 1 RCIM_Constant>,<at 0.85 0>,<at 1 0>,<at 1.15 1.4>,<at 1.35 1>,<at 1.95 1>,<at 2 0>,<at 2.15 1.4>,<at 2.35 1>,<at 2.8325 1>,<at 3 0>]}
       }}
      CurveIndex=&Add_Float(
       A=$Particles.NormalizedAge
       B=&MakeFloatFromInt(Int=&RandomRangeInt(Minimum=1, Maximum=2))
      )
      ScaleCurve=<1 1>
     )
     Passthrough Vector2D=&Vector2DFromCurve(
      Vector2Curve=DataInterface() {propertyValues {XCurve {keys [<at 0 0.7459>,<at 0.1327 1>,<at 1 0.3049>]}, YCurve {keys [<at 0 0.8721>,<at 0.5 1>,<at 1 0.1672>]}}}
      ScaleCurve=<1 1>
     )
     Spawn Group Mask=0
    )) {isSetParameters=true, module=SetVariables_#24}
   &ScaleMeshSize(Scale Factor=&MakeVector(
     X=&MakeFloatFromVector2d(Vector 2D=$Particles.MeshScale2D)
     Y=&MakeFloatFromVector2d(Vector 2D=$Particles.MeshScale2D)
     Z=&MakeFloatFromVector2d(Vector 2D=$Particles.MeshScale2D, Channel=::NewEnumerator1)
    ))
   &GravityForce(Gravity=&Multiply_VectorByFloat(
     Vector=<0 0 -200>
     Float=&MaskFloatBySpawnGroup(Masked Float=0, Passthrough Float=&RandomRangeFloat(Minimum=0.5, Maximum=1))
    ))
   &UpdateMeshOrientation(
    Rotation Rate=&MaskFloatBySpawnGroup(Masked Float=0, Passthrough Float=0.4)
    Rotation Vector=&RandomVector(Vector Scale=1)
   )
   &Drag(Drag=3, Rotational Drag=3, Use Linear Drag=true, Use Rotational Drag=true)
   &SolveForcesAndVelocity(Clamp Velocity=false, Limit Acceleration=false)
   /SetVariables(
    Particles.GlowColor=$System.SystemCrystalCol
    Particles.GlowOffset=8
    Particles.GlowParam:Vector4f=<0.2 1 0 0>
    Particles.GlowSize=&Multiply_Vector2D(
     A=&MakeVector2D(
      X=&MakeFloatFromVector(Vector=$Particles.Scale, Channel=::NewEnumerator0)
      Y=&MakeFloatFromVector(Vector=$Particles.Scale, Channel=::NewEnumerator2)
     )
     B=<21.7595 29.7918>
    )
   ) {isSetParameters=true, module=SetVariables_#25}
   /SetVariables(Particles.FlickerColor:LinearColor=&Add_LinearColor(
     A:LinearColor=&LerpLinearColors(
      StartColor=&Multiply_LinearColorByFloat(Linear Color=$System.SystemCrystalCol, Float=&RandomRangeFloat(Minimum=0.7, Maximum=1))
      EndColor=&Multiply_LinearColorByFloat(Linear Color=$System.SystemTrimCol, Float=1)
      LerpFactor=&FloatFromCurve(
       FloatCurve=DataInterface() {propertyValues {Curve {keys [
           {arriveTangent=5.9013, interpMode=RCIM_Cubic, leaveTangent=5.9013, tangentMode=RCTM_User, time=0, value=0.3676}
           {arriveTangent=-1.4272, interpMode=RCIM_Cubic, leaveTangent=-1.4272, tangentMode=RCTM_User, time=0.1102, value=1}
           {arriveTangent=-0.8994, tangentMode=RCTM_User, time=0.2921, value=0}
           <at 1 0 RCIM_Cubic>
          ]}}}
       Scale Curve=&RandomRangeFloat(Minimum=0.3, Maximum=1)
      )
     )
     B:LinearColor=&MakeLinearColorFromFloat(Value:NiagaraFloat=&FloatFromCurve(
       FloatCurve=DataInterface() {propertyValues {Curve {keys [<at 0 0>,<at 0.1754 0.6533>,<at 0.2838 0>,<at 0.345 0.2634>,<at 0.4303 0.149>,<at 0.5361 0.4843>,<at 0.7476 0>,<at 1.2452 0>,<at 1.3259 0.2307>,<at 1.4591 0>,<at 1.5391 0.6635>,<at 1.6457 0>,<at 1.7576 0.4348>,<at 1.8323 0>,<at 2.1468 0>,<at 2.36 0.5083>,<at 2.8611 0>,<at 3 0>]}}}
       CurveIndex=&Add_Float(A=$Particles.NormalizedAge, B=&RandomRangeFloat(Minimum=-0.2, Maximum=2))
       Scale Curve=&RandomRangeFloat(Minimum=0.3, Maximum=1)
      ))
    )) {isSetParameters=true, module=SetVariables_#26}
   /SetVariables(Particles.SpikePosition:NiagaraPosition=&Add_Vector(
     A=&Multiply_VectorByFloat(
      Vector=$Particles.DirectionRemapped
      Float=&Add_Float(
       A=&RandomRangeFloat(Minimum=9, Maximum=26)
       B=&FloatFromCurve(
        FloatCurve=DataInterface() {propertyValues {Curve {keys [
            {arriveTangent=2.2394, interpMode=RCIM_Cubic, leaveTangent=2.2394, tangentMode=RCTM_User, time=0, value=-1}
            {arriveTangent=0.1323, leaveTangent=1, tangentMode=RCTM_User, time=0.8879, value=0}
            <at 1 0>
           ]}}}
        CurveIndex=&Add_Float(A=$Particles.NormalizedAge, B=&RandomRangeFloat(Minimum=0, Maximum=0.3))
        Scale Curve=&RandomRangeFloat(Minimum=20, Maximum=40)
       )
      )
     )
     B=<4.6408 0 0>
    )) {isSetParameters=true, module=SetVariables_#27}
  ]
  rendererClasses [&NiagaraMeshRendererProperties, &NiagaraSpriteRendererProperties]
  renderers [
   {rendererClass=&NiagaraMeshRendererProperties, rendererIndex=0}
   {rendererClass=&NiagaraMeshRendererProperties, rendererIndex=1}
   {rendererClass=&NiagaraSpriteRendererProperties, rendererIndex=2}
   {rendererClass=&NiagaraSpriteRendererProperties, rendererIndex=3}
  ]
 }
 Sparks004 {
  bEnabled=false
  emitterUpdateScript [
   &EmitterState()
   &SpawnBurst_Instantaneous(
    Spawn Count=&RandomRangeInt(Minimum=4, Maximum=5)
    Spawn Probability=&MakeCustomFloatFromBool(Bool=$User.bIsHardSurface, True Float=1, False Float=0)
    Use Spawn Probability=true
    UseLoopCountLimit=false
   )
   &SpawnBurst_Instantaneous(
    Spawn Count=50
    Spawn Probability=&MakeCustomFloatFromBool(Bool=$User.bIsHardSurface, True Float=1, False Float=0)
    Use Spawn Probability=true
    UseLoopCountLimit=false
   ) {enabled=false, module=SpawnBurst_Instantaneous002}
   &SpawnBurst_Instantaneous(
    Spawn Count=&MakeCustomIntFromBool(Bool=$User.bIsFrontEndPreview, True INT=30, False INT=&RandomRangeInt(Minimum=15, Maximum=20))
    Spawn Probability=&MakeCustomFloatFromBool(Bool=$User.bIsHardSurface, True Float=1, False Float=0)
    Use Spawn Probability=false
    UseLoopCountLimit=false
   ) {enabled=false, module=SpawnBurst_Instantaneous001}
  ]
  particleSpawnScript [
   &InitializeParticle(
    Color Maximum=<rgba 1 0.0003 0.0776 3>
    Color Minimum=<rgba 1 0 0.046 1>
    Color Mode=::NewEnumerator2
    Lifetime Max=0.3
    Lifetime Min=0.2
    Lifetime Mode=::NewEnumerator1
    Mesh Scale=<1.5 0.3 0.3>
    Mesh Scale Mode=::NewEnumerator4
    UsePositionOffset=false
    Write Material Random=true
   )
   /SetVariables(Particles.DirectionalVel:NiagaraBool=&SetBoolByFloatComparison(
     A=&FloatFromCurve(
      FloatCurve=DataInterface() {propertyValues {Curve {defaultValue=1, keys [{tangentMode=RCTM_User, time=0, value=0}, <at 0.3371 0.5>, {tangentMode=RCTM_User, time=1, value=1}]}}}
      CurveIndex=&RandomRangeFloat(Minimum=0, Maximum=1)
     )
     B=0.5
    )) {enabled=false, isSetParameters=true}
   &ShapeLocation(
    Hemisphere Distribution=<1 0.1>
    MaskSpawnGroup=false
    Override Seed=false
    Rotation Axis=&Add_Vector(A=$System.DirectionVector, B=<0.5 0 0>)
    Rotation Mode=::NewEnumerator1
    Sphere Radius=&RandomRangeFloat(Minimum=5, Maximum=15)
    Sphere Surface Distribution=1
    UseOwnerScale=true
   )
   &ShapeLocation(
    Hemisphere Distribution=<1 0.5>
    MaskSpawnGroup=false
    Offset=&Multiply_VectorByFloat(Vector=$System.DirectionVector, Float=15)
    Offset Coordinate Space=::Local
    Offset Mode=::NewEnumerator0
    Override Seed=false
    Rotation Mode=::NewEnumerator1
    Sphere Radius=3
    UseOwnerScale=true
   ) {enabled=false, module=ShapeLocation001}
   &CameraOffset(Camera Offset Amount=40)
   &AddVelocity(
    Module.Use Falloff Distance=false
    Override Seed=false
    Velocity Mode=::NewEnumerator1
    Velocity Speed=&RandomRangeFloat(Minimum=400, Maximum=700)
   )
   &AddVelocity(
    Cone Angle=60
    Cone Axis=&Add_Vector(A=$System.DirectionVector, B=<1 0 0>)
    Module.UseConeFalloff=false
    Module.UseVelDistribution=false
    Override Seed=false
    Velocity Mode=::NewEnumerator2
    Velocity Speed=&RandomRangeFloat(Minimum=400, Maximum=700)
   ) {enabled=false, module=AddVelocity002}
   &AddVelocity(
    Velocity=&Add_Vector(A=$System.DirectionVector, B=&RandomRangeVector(Minimum=<-0.1 -0.1 -0.1>, Maximum=<0.1 0.1 0.1>))
    Velocity Speed Scale=&RandomRangeFloat(Minimum=20, Maximum=500)
   ) {enabled=false, module=AddVelocity001}
   /SetVariables(
    Particles.CrystalDynam:Vector4f=&MakeVector4_FromLinearColor(LinearColor=<rgba 0.5652 0 0 1>)
    Particles.MeshIndex=&RandomRangeInt(Minimum=0, Maximum=2)
   ) {isSetParameters=true, module=SetVariables_#21}
   &InitialMeshOrientation(
    Mesh Axis To Orient=<-1 0 0>
    Mesh Orientation Mode=::NewEnumerator2
    Orientation Coordinate Space=::Local
    Orientation Vector=$Particles.Velocity
    Rotation=&RandomRangeVector(Minimum=<0 0 0>, Maximum=<1 0 0>)
    Use Rotation Vector=true
   )
  ]
  particleUpdateScript [
   &ParticleState()
   &Drag(Drag=5, Rotational Drag=5, Use Linear Drag=true, Use Rotational Drag=true)
   &ScaleMeshSize(Scale Factor=&VectorFromCurve(VectorCurve=DataInterface() {propertyValues {XCurve {keys [<at 0 0>,<at 0.15 1>]}, YCurve {keys [<at 0.75 1>,<at 1 0>]}, ZCurve {keys [<at 0.75 1>,<at 1 0>]}}}))
   &AccelerationForce(Acceleration=&Multiply_VectorByFloat(
     Vector=&RandomRangeVector(Minimum=<0 0 -30>, Maximum=<0 0 -150>)
     Float=&MakeCustomFloatFromBool(Bool=$Particles.Initial.DirectionalVel, True Float=3, False Float=0.1)
    )) {enabled=false}
   &SolveForcesAndVelocity(Clamp Velocity=false, Limit Acceleration=false)
   /SetVariables(Particles.GlowColor=<rgba 0 1 0.994 50>, Particles.GlowSize=<10 10>) {isSetParameters=true}
  ]
  rendererClasses [&NiagaraSpriteRendererProperties, &NiagaraMeshRendererProperties]
  renderers [
   {rendererClass=&NiagaraSpriteRendererProperties}
   {rendererClass=&NiagaraSpriteRendererProperties, rendererIndex=1}
   {rendererClass=&NiagaraMeshRendererProperties, rendererIndex=2}
  ]
 }
 Sparks_CPU {
  bEnabled=false
  emitterUpdateScript [
   &EmitterState(
    Enable Distance Culling=true
    Max Distance Response=::KillImmediately
    Module.bApplySpawnCountScaleByDistance=true
    Module.bUseMaxDistance=true
    Module.bUseMinDistance=false
    Scalability Mode=::NewEnumerator1
    Scale Spawn Count=true
    Spawn Count Scale By Distance=DataInterface() {propertyValues {Curve {keys [<at 0.3676 0.9657>,<at 0.9962 0.4982>]}}}
   )
   &SpawnBurst_Instantaneous(Spawn Count=&UniformRangedInt(Minimum=5, Maximum=10), Use Spawn Probability=false)
   &SpawnRate(
    Spawn Probability=&FloatFromCurve(
     FloatCurve=DataInterface() {propertyValues {Curve {keys [{arriveTangent=-12.2872, interpMode=RCIM_Cubic, leaveTangent=-12.2872, tangentMode=RCTM_User, time=0, value=1}, <at 0.1 0>]}}}
     CurveIndex=$System.Age
    )
    SpawnRate=111
    Use Spawn Probability=true
   )
  ]
  particleSpawnScript [
   &InitializeParticle(
    Color=<rgba 3 3 3 1>
    Lifetime Max=0.45
    Lifetime Min=0.166
    Lifetime Mode=::NewEnumerator1
    Sprite Size Mode=::NewEnumerator3
    Uniform Sprite Size=&Multiply_Float(
     A=&MakeCustomFloatFromBool(Bool=$User.bIsFrontEnd, True Float=0.45, False Float=1)
     B=&RandomRangeFloat(Minimum=2.5, Maximum=7)
    )
    UsePositionOffset=false
    Write Material Random=true
   )
   &ShapeLocation(
    Hemisphere Distribution=&FN_FrontEndVector2Modifier(Front End:Vector2f=<1 1>, Gameplay:Vector2f=<1 0.5>)
    MaskSpawnGroup=false
    Override Seed=false
    Rotation Mode=::NewEnumerator1
    Sphere Radius=9
    Sphere Surface Distribution=1
    UseOwnerScale=true
   )
   &AddVelocity(
    Module.Use Falloff Distance=false
    Override Seed=false
    Velocity Mode=::NewEnumerator1
    Velocity Speed=&FN_FrontEndFloatModifier(
     Gameplay=&RandomRangeFloat(Minimum=222, Maximum=666)
     Front End=&RandomRangeFloat(Minimum=150, Maximum=400)
    )
   )
   &DynamicMaterialParameters(
    Index 0 Param 1=3
    Index 0 Param 2=20
    Index 0 Param 3=1
    Module.Param0WriteEnabled=true
    Module.Param1WriteEnabled=true
    Module.Param2WriteEnabled=true
    Module.Param3WriteEnabled=true
   )
  ]
  particleUpdateScript [
   &ParticleState()
   &GravityForce(Gravity=<0 0 -666>)
   &Drag(Drag=&UniformRangedFloat(Minimum=2, Maximum=6), Use Linear Drag=true, Use Rotational Drag=false)
   &SolveForcesAndVelocity(Clamp Velocity=false, Limit Acceleration=false)
   &ScaleSpriteSize(Scale Factor=&Vector2DFromCurve(
     Vector2Curve=DataInterface() {propertyValues {
       XCurve {keys [
         {arriveTangent=5.9728, interpMode=RCIM_Cubic, leaveTangent=5.9728, tangentMode=RCTM_User, time=0, value=0.1}
         {arriveTangent=-0.1, interpMode=RCIM_Cubic, leaveTangent=-0.1, tangentMode=RCTM_User, time=0.18, value=0.8}
         {interpMode=RCIM_Cubic, tangentMode=RCTM_User, time=1, value=0}
        ]}
       YCurve {keys [
         {arriveTangent=6.2715, interpMode=RCIM_Cubic, leaveTangent=6.2715, tangentMode=RCTM_User, time=0, value=0.1}
         {arriveTangent=-0.1001, interpMode=RCIM_Cubic, leaveTangent=-0.1001, tangentMode=RCTM_User, time=0.18, value=0.8}
         {interpMode=RCIM_Cubic, tangentMode=RCTM_User, time=0.9988, value=0}
        ]}
      }}
     ScaleCurve=<1 1>
    ))
   &ScaleSpriteSizeBySpeed(Max Scale Factor=<0.5 8>, Min Scale Factor=<0.5 1>, Velocity Threshold=666)
   &ScaleColor(
    Linear Color Curve=DataInterface() {propertyValues {
      AlphaCurve {keys [<at 0.333 1>,<at 1 0>]}
      BlueCurve {keys [<at 0 0.24>,<at 0.166 0.2>,<at 0.333 0>,<at 0.75 0>]}
      GreenCurve {keys [<at 0 3.12>,<at 0.166 2.6>,<at 0.333 0.6667>,<at 0.75 0.15>]}
      RedCurve {keys [<at 0 6>,<at 0.166 5>,<at 0.333 4>,<at 0.75 3>]}
     }}
    Scale Mode=::NewEnumerator2
   )
  ]
  rendererClasses [&NiagaraSpriteRendererProperties]
  renderers [{rendererClass=&NiagaraSpriteRendererProperties}]
 }
 Sparks_Gpu {
  bEnabled=false
  simTarget=GPUComputeSim
  emitterUpdateScript [
   &EmitterState()
   &SpawnBurst_Instantaneous(
    Spawn Count=&UniformRangedInt(Minimum=6, Maximum=15)
    Spawn Time=&MakeCustomFloatFromBool(Bool=$User.bIsHardSurface, True Float=1, False Float=0)
    Use Spawn Probability=false
   )
   &SpawnRate(
    Spawn Probability=&FloatFromCurve(
     FloatCurve=DataInterface() {propertyValues {Curve {keys [{arriveTangent=-12.2872, interpMode=RCIM_Cubic, leaveTangent=-12.2872, tangentMode=RCTM_User, time=0, value=1}, <at 0.1 0>]}}}
     CurveIndex=$System.Age
     Scale Curve=&MakeCustomFloatFromBool(Bool=$User.bIsHardSurface, True Float=1, False Float=0)
    )
    SpawnRate=100
    Use Spawn Probability=true
   )
  ]
  particleSpawnScript [
   &InitializeParticle(
    Color Mode=::NewEnumerator0
    Lifetime=&FN_FrontEndFloatModifier(
     Gameplay=&RandomRangeFloat(Minimum=0.25, Maximum=1.1)
     Front End=&RandomRangeFloat(Minimum=0.25, Maximum=0.5)
    )
    Sprite Size Mode=::NewEnumerator3
    Uniform Sprite Size=&Multiply_Float(
     A=&MakeCustomFloatFromBool(Bool=$User.bIsFrontEnd, True Float=0.45, False Float=1)
     B=&FloatFromCurve(
      FloatCurve=DataInterface() {propertyValues {Curve {keys [<at 0 13>,<at 0.2 7>,<at 1 2.5>]}}}
      CurveIndex=&RandomRangeFloat(Minimum=0, Maximum=1)
     )
    )
    UsePositionOffset=false
    Write Material Random=true
   )
   &ShapeLocation(
    Hemisphere Distribution=<1 0.5>
    MaskSpawnGroup=false
    Override Seed=false
    Rotation Angle=90
    Rotation Mode=::NewEnumerator1
    Sphere Radius=15
    Sphere Surface Distribution=1
    UseOwnerScale=true
   )
   &AddVelocity(
    Module.Use Falloff Distance=false
    Override Seed=false
    Velocity Mode=::NewEnumerator1
    Velocity Speed=&FN_FrontEndFloatModifier(
     Gameplay=&RandomRangeFloat(Minimum=100, Maximum=350)
     Front End=&RandomRangeFloat(Minimum=80, Maximum=250)
    )
   )
   &CurlNoiseForce(
    Noise Frequency=75
    Noise Strength=&RandomRangeFloat(Minimum=100, Maximum=250)
    Pan Noise=true
    Pan Noise Field=<2 2 0>
   )
   &ApplyInitialForces()
   &DynamicMaterialParameters(
    Index 0 Param 1=3
    Index 0 Param 2=20
    Index 0 Param 3=1
    Module.Param0WriteEnabled=true
    Module.Param1WriteEnabled=true
    Module.Param2WriteEnabled=true
    Module.Param3WriteEnabled=true
   )
  ]
  particleUpdateScript [
   &ParticleState()
   &GravityForce()
   &Drag(Drag=&UniformRangedFloat(Minimum=1.5, Maximum=3), Use Linear Drag=true, Use Rotational Drag=false)
   &Collision(
    Age Colliding Particles=false
    Kill Offscreen Particles=true
    KillOnCollision=false
    Module.Control Rest State Via Surface Normal=false
    Module.EnableMaxCollisionCount=false
    Module.Randomize Collision Normal Vector=false
   )
   &SolveForcesAndVelocity(Clamp Velocity=false, Limit Acceleration=false)
   &ScaleSpriteSize(Scale Factor=&Vector2DFromCurve(
     Vector2Curve=DataInterface() {propertyValues {
       XCurve {keys [
         {arriveTangent=5.9728, interpMode=RCIM_Cubic, leaveTangent=5.9728, tangentMode=RCTM_User, time=0, value=0.1}
         {arriveTangent=-0.1, interpMode=RCIM_Cubic, leaveTangent=-0.1, tangentMode=RCTM_User, time=0.18, value=0.8}
         {interpMode=RCIM_Cubic, tangentMode=RCTM_User, time=1, value=0}
        ]}
       YCurve {keys [
         {arriveTangent=6.2715, interpMode=RCIM_Cubic, leaveTangent=6.2715, tangentMode=RCTM_User, time=0, value=0.1}
         {arriveTangent=-0.1001, interpMode=RCIM_Cubic, leaveTangent=-0.1001, tangentMode=RCTM_User, time=0.18, value=0.8}
         {interpMode=RCIM_Cubic, tangentMode=RCTM_User, time=0.9988, value=0}
        ]}
      }}
     ScaleCurve=<1 1>
    ))
   &ScaleSpriteSizeBySpeed(Max Scale Factor=<0.5 3>, Min Scale Factor=<0.5 1>, Velocity Threshold=666)
   &ScaleColor(
    Linear Color Curve=DataInterface() {propertyValues {
      AlphaCurve {keys [<at 0.333 1>,<at 1 0>]}
      BlueCurve {keys [<at 0 0.24>,<at 0.166 0.2>,<at 0.333 0>,<at 0.75 0>]}
      GreenCurve {keys [<at 0 3.12>,<at 0.166 2.6>,<at 0.333 0.6667>,<at 0.75 0.15>]}
      RedCurve {keys [<at 0 6>,<at 0.166 5>,<at 0.333 4>,<at 0.75 3>]}
     }}
    Scale Mode=::NewEnumerator2
   )
  ]
  rendererClasses [&NiagaraSpriteRendererProperties]
  renderers [{rendererClass=&NiagaraSpriteRendererProperties}]
 }
}
summary {
 systemName=NS_Pickaxe_NobleCrusher_Impact
 emitters [
  {
   emitterName=Flash_Elements
   rendererClasses [
    &NiagaraSpriteRendererProperties
    &NiagaraLightRendererProperties
    &NiagaraMeshRendererProperties
   ]
  }
  {bEnabled=false, emitterName=Sparks_CPU, rendererClasses [&NiagaraSpriteRendererProperties]}
  {bEnabled=false, emitterName=Sparks_Gpu, simTarget=GPUComputeSim, rendererClasses [&NiagaraSpriteRendererProperties]}
  {bEnabled=false, emitterName=Impact_Smoke, rendererClasses [&NiagaraSpriteRendererProperties]}
  {
   bEnabled=false
   emitterName=FN_PhysMatDebris
   rendererClasses [&NiagaraSpriteRendererProperties, &NiagaraMeshRendererProperties]
  }
  {emitterName=Sparks, rendererClasses [&NiagaraMeshRendererProperties, &NiagaraSpriteRendererProperties]}
  {emitterName=Sparks001, rendererClasses [&NiagaraMeshRendererProperties, &NiagaraSpriteRendererProperties]}
  {bEnabled=false, emitterName=Impact_Smoke001, rendererClasses [&NiagaraSpriteRendererProperties]}
  {emitterName=Mist, rendererClasses [&NiagaraSpriteRendererProperties]}
  {emitterName=Sparkle, rendererClasses [&NiagaraSpriteRendererProperties]}
  {emitterName=Ring, rendererClasses [&NiagaraSpriteRendererProperties, &NiagaraMeshRendererProperties]}
  {
   bEnabled=false
   emitterName=Sparks004
   rendererClasses [&NiagaraSpriteRendererProperties, &NiagaraMeshRendererProperties]
  }
  {
   bEnabled=false
   emitterName=Sparks002
   rendererClasses [&NiagaraMeshRendererProperties, &NiagaraSpriteRendererProperties]
  }
  {
   bEnabled=false
   emitterName=Sparks003
   rendererClasses [&NiagaraMeshRendererProperties, &NiagaraSpriteRendererProperties]
  }
  {
   bEnabled=false
   emitterName=Ring001
   rendererClasses [&NiagaraSpriteRendererProperties, &NiagaraMeshRendererProperties]
  }
 ]
 userVariables [
  User.bIsFrontEnd {type=NiagaraBool}
  User.bIsFrontEndPreview {type=NiagaraBool}
  User.bIsHardSurface {type=NiagaraBool, defaultValue=true}
  User.Crystal_Color {type=LinearColor, description="", defaultValue {struct=LinearColor, value=<rgba 0.4342 0.0369 0.0091 1>}}
  User.PhysicalSurface {type=NiagaraInt32}
  User.SwingSectionIndex {type=NiagaraInt32}
  User.Trim_Color {type=LinearColor, description="", defaultValue {struct=LinearColor, value=<rgba 0.1878 0.1878 0.1878 0>}}
 ]
}
